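# Niagara system for the Drift Horror male glider cosmetic, layering pink-magenta energy ribbons, lightning, dark smoke, ribbon tails and pendulum drips.
# Bone-flow ribbons follow filtered sockets on the glider mesh, while smoke samples its skeletal mesh.
# Self-contained bone-lookup modules place flickering lightning meshes and lights, with arcs stretched between randomly lerped bone pairs.
# Ribbon tails ride Hermite splines built from time-smoothed bone positions; several alternate emitters are disabled.
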NiagaraSystem /BRCosmetics/Gadgets/Assets/VinderTech_GliderChute/Glider_Drift_Horror_Male/FX/NS_DriftHorror_Glider
dependencies {
 usedDataInterfaces [
  &NiagaraDataInterfaceArrayInt32
  &NiagaraDataInterfaceCurve
  &NiagaraDataInterfaceColorCurve
  &NiagaraDataInterfaceVectorCurve
  &NiagaraDataInterfaceVector2DCurve
  &NiagaraDataInterfaceArrayFloat
  &NiagaraDataInterfaceArrayFloat2
  &NiagaraDataInterfaceArrayFloat3
  &NiagaraDataInterfaceArrayColor
 ]
 usedDynamicInputs [
  &RandomRangeVector2D
  &Abs_Float
  &Boolean_OrOperation
  &Boolean_AndOperation
 ]
 usedModules [&SystemState, &Timeline]
}
emitters {
 Bone_Flow_Ribbon02 {
  emitterUpdateScript [
   &EmitterState(Life Cycle Mode=::NewEnumerator1, Loop Duration=0.1, UseLoopDelay=true)
   &SpawnBurst_Instantaneous(Spawn Count=6, Use Spawn Probability=false, UseLoopCountLimit=false)
   /SetVariables(
    Emitter.gen rand int:NiagaraInt32=&RandomRangeInt(Minimum=0, Maximum=10, Evaluation Type=::NewEnumerator1)
    Emitter.genRand:NiagaraFloat=&RandomRangeFloat(Minimum=0, Maximum=1, Evaluation Type=::NewEnumerator1)
   ) {isSetParameters=true}
  ]
  particleSpawnScript [
   &FN_SetSpawnGroup(Inclusivity=true, Normalize Spawn Group=true, Particles Per Group=6, Pattern=::NewEnumerator0, Set Ribbon ID and Ribbon Link Order=true)
   &InitializeRibbon(Color=<rgba 3 0 1.916 1>, Lifetime=0.35, Ribbon Width=55, UsePositionOffset=false, Write Material Random=true)
   &CameraOffset(Camera Offset Amount=10)
   &DynamicMaterialParameters(
    Index 0 Param 1=0.5
    Index 0 Param 3=&RandomRangeFloat(Minimum=0.3, Maximum=0.5)
    Index 0 Param 4=1
    Module.Param0WriteEnabled=true
    Module.Param1WriteEnabled=true
    Module.Param2WriteEnabled=true
    Module.Param3WriteEnabled=true
   )
   /SetVariables(Particles.random:NiagaraFloat=$Emitter.genRand, Particles.randomInt:NiagaraInt32=$Emitter.gen rand int) {isSetParameters=true}
  ]
  particleUpdateScript [
   &ParticleState()
   &FN_BoneFlow(
    Angle:NiagaraFloat=&Add_Float(A=$Particles.random, B=&RandomRangeFloat(Minimum=0.08, Maximum=-0.08))
    BoneID:NiagaraInt32=$Particles.Initial.randomInt
    Height:NiagaraFloat=&Lerp_Float(
     A=&RandomRangeFloat(Minimum=0, Maximum=0.25)
     B=&RandomRangeFloat(Minimum=0.75, Maximum=1)
     Alpha=$Particles.Normalized Spawn Group
    )
    Radius:NiagaraFloat=&Multiply_Float(
     A=&FloatFromCurve(
      FloatCurve=DataInterface() {propertyValues {Curve {keys [
          {arriveTangent=4.3816, interpMode=RCIM_Cubic, leaveTangent=4.3816, tangentMode=RCTM_User, time=0, value=0}
          {interpMode=RCIM_Cubic, tangentMode=RCTM_User, time=0.5, value=1}
          {arriveTangent=-4.6838, interpMode=RCIM_Cubic, leaveTangent=-4.6838, tangentMode=RCTM_User, time=1, value=0}
         ]}}}
      CurveIndex=$Particles.Normalized Spawn Group
      Scale Curve=25
     )
     B=&RandomRangeFloat(Minimum=1, Maximum=1)
    )
    Set Material Dynamic=false
    SkeletalMesh=DataInterface() {propertyValues {
      DefaultMesh=null
      ExcludeBoneName=root
      OptionalComponentClass=null
      PreviewMesh=/BRCosmetics/Gadgets/Assets/VinderTech_GliderChute/Glider_Drift_Horror_Male/Meshes/Drift_Horror_Male_Glider
      SkinningMode=SkinOnTheFly
      SoftSourceActor=null
      SourceMode=Default
      UvSetIndex=0
      WholeMeshLOD=-1
      bExcludeBone=false
      bReadDeformedGeometry=true
      bRequireCurrentFrameData=true
      ComponentTags []
      FilteredBones []
      FilteredSockets [
       FX Socket Flow L1 0
       FX Socket Flow L1 1
       FX Socket Flow L1 2
       FX Socket Flow L1 2
       FX Socket Flow L1 3
       FX Socket Flow L1 4
       FX Socket Flow L1 4
       FX Socket Flow L1 5
       FX Socket Flow L1 6
       FX Socket Flow L2 0
       FX Socket Flow L2 1
       FX Socket Flow L2 2
       FX Socket Flow L2 2
       FX Socket Flow L2 3
       FX Socket Flow L2 4
       FX Socket Flow L3 0
       FX Socket Flow L3 1
       FX Socket Flow L3 2
       FX Socket Flow R3 0
       FX Socket Flow R3 1
       FX Socket Flow R3 2
       FX Socket Flow R2 0
       FX Socket Flow R2 1
       FX Socket Flow R2 2
       FX Socket Flow R1 0
       FX Socket Flow R1 1
       FX Socket Flow R1 2
       FX Socket Flow R0 0
       FX Socket Flow R0 1
       FX Socket Flow R0 2
       FX Socket Flow R0 1
       FX Socket Flow R0 2
       FX Socket Flow R0 4
      ]
      MeshUserParameter {parameter {name=None, typeDefHandle {registeredTypeIndex=1}, varData []}}
      SamplingRegions []
     }}
    Use Sockets Instead=true
   )
   /SetVariables(Particles.RibbonFacing=$Output.FN_BoneFlow.SplineNormal) {isSetParameters=true}
   &ScaleRibbonWidth(Ribbon Width Scale=&Multiply_Float(
     A=&Multiply_Float(
      A=&FloatFromCurve(
       FloatCurve=DataInterface() {propertyValues {Curve {keys [
           {interpMode=RCIM_Cubic, tangentMode=RCTM_User, time=0, value=0}
           <at 0.1 0>
           {interpMode=RCIM_Cubic, tangentMode=RCTM_Break, time=0.2, value=1}
           {arriveTangent=-1.5728, interpMode=RCIM_Cubic, leaveTangent=-1.5728, tangentMode=RCTM_User, time=1, value=0.5}
          ]}}}
       Scale Curve=&RandomRangeFloat(Minimum=0.7, Maximum=1)
      )
      B=1
     )
     B=&FloatFromCurve(
      FloatCurve=DataInterface() {propertyValues {Curve {keys [
          {interpMode=RCIM_Cubic, tangentMode=RCTM_User, time=0, value=0.675}
          {interpMode=RCIM_Cubic, tangentMode=RCTM_User, time=0.5, value=1}
          {tangentMode=RCTM_User, time=1, value=0.675}
         ]}}}
      CurveIndex=$Particles.Normalized Spawn Group
      Scale Curve=&FloatFromCurve(
       FloatCurve=DataInterface() {propertyValues {Curve {keys [
           {interpMode=RCIM_Cubic, tangentMode=RCTM_User, time=0, value=1}
           {interpMode=RCIM_Cubic, tangentMode=RCTM_User, time=0.3, value=0.4}
           {interpMode=RCIM_Cubic, tangentMode=RCTM_User, time=0.5, value=1.5}
           {interpMode=RCIM_Cubic, tangentMode=RCTM_User, time=0.7, value=0.4}
           {interpMode=RCIM_Cubic, tangentMode=RCTM_User, time=1, value=1}
          ]}}}
       CurveIndex=&Multiply_Float(
        A=$Particles.Normalized Spawn Group
        B=&Multiply_Float(A=$Particles.Age, B=&RandomRangeFloat(Minimum=1, Maximum=1.5))
       )
      )
     )
    ))
   &ScaleColor(
    Scale Mode=::NewEnumerator1
    Scale RGBA=&Vector4FromCurve(Vector4Curve=DataInterface() {propertyValues {
       AlphaCurve {keys [
         <at 0 0 RCIM_Cubic>
         {arriveTangent=2.0471, interpMode=RCIM_Cubic, leaveTangent=2.0471, time=0.2262, value=1}
         {arriveTangent=-1.2924, interpMode=RCIM_Cubic, leaveTangent=-1.2924, time=0.4885, value=1}
         <at 1 0 RCIM_Cubic>
        ]}
       BlueCurve {keys [<at 0.3 1>]}
       GreenCurve {keys [<at 0.3 1>]}
       RedCurve {keys [<at 0.3 1>]}
      }})
    ScaleRGBA=true
   ) {module=ScaleColor001}
   &SolveForcesAndVelocity(Clamp Velocity=false, Limit Acceleration=false)
  ]
  rendererClasses [&NiagaraRibbonRendererProperties]
  renderers [{rendererClass=&NiagaraRibbonRendererProperties}]
 }
 Drips_Falling {
  bEnabled=false
  emitterUpdateScript [
   &EmitterState()
   &SpawnRate(SpawnRate=5, Use Spawn Probability=false) {module=SpawnRate001}
  ]
  particleSpawnScript [
   &InitializeParticle(
    Color Mode=::NewEnumerator0
    Lifetime=1.5
    Sprite Size Max=<3 5>
    Sprite Size Mode=::NewEnumerator2
    UsePositionOffset=false
    Write Material Random=true
   )
   &SkeletalMeshLocation(
    Bone Sampling Mode=::NewEnumerator2
    MaskSpawnGroup=false
    OverrideRandomSeed=false
    Sampled Position Offset Coordinate Space:ENiagaraSkelMeshTransforms=::NewEnumerator1
    Skeletal Mesh=$User.SkeletalMesh
   )
  ]
  particleUpdateScript [
   &ParticleState()
   &InheritSourceMovement(ApplyForce=false, ApplyPosition=true, ApplyPositionFalloff=false)
   &GravityForce(Gravity=&Multiply_VectorByFloat(
     Vector=<0 0 -980>
     Float=&FloatFromCurve(FloatCurve=DataInterface() {propertyValues {Curve {keys [
          {interpMode=RCIM_Cubic, tangentMode=RCTM_User, time=0, value=0}
          {arriveTangent=1.3497, interpMode=RCIM_Cubic, leaveTangent=1.3497, tangentMode=RCTM_User, time=1, value=1}
         ]}}})
    ))
   &SolveForcesAndVelocity(Clamp Velocity=false, Limit Acceleration=false)
   &GenerateLocationEvent(Event Send Rate=60, Use Event Probability=false, UseEventDelay=false)
   &ScaleColor(ScaleA=true, ScaleRGB=true)
   &ScaleSpriteSize(Uniform Curve Sprite Scale=DataInterface() {propertyValues {Curve {keys [
        <at 0 0 RCIM_Cubic>
        <at 0.15 1 RCIM_Cubic>
        {arriveTangent=-3.5189, interpMode=RCIM_Cubic, leaveTangent=-3.5189, tangentMode=RCTM_User, time=1, value=0}
       ]}}})
   &ScaleSpriteSizeBySpeed(Min Scale Factor=<0.75 0.5>, Velocity Threshold=1500)
  ]
  rendererClasses [&NiagaraSpriteRendererProperties]
  renderers [{rendererClass=&NiagaraSpriteRendererProperties}]
 }
 Lightning_Spike_Arcs {
  emitterUpdateScript [
   &EmitterState(
    Life Cycle Mode=::NewEnumerator1
    Loop Duration=&RandomRangeFloat(Minimum=0, Maximum=1)
    Recalculate Duration Each Loop=true
    UseLoopDelay=false
   )
   &SpawnBurst_Instantaneous(
    Spawn Count=&MakeCustomIntFromBool(
     Bool=&SetBoolByFloatComparison(A=&RandomRangeFloat(Minimum=0, Maximum=1), B=0.25)
     True INT=1
     False INT=2
    )
    Use Spawn Probability=false
    UseLoopCountLimit=false
   )
   &SpawnRate(SpawnRate=500, Use Spawn Probability=false) {enabled=false}
  ]
  particleSpawnScript [
   &InitializeParticle(
    Color=<rgba 2 0 1.5333 1>
    Lifetime=0.25
    Mesh Scale Max=<0.037 0.5 0.5>
    Mesh Scale Min=<0.037 1.25 1.25>
    Mesh Scale Mode=::NewEnumerator2
    Position Mode=::NewEnumerator1
    Sprite Size Mode=::NewEnumerator3
    Uniform Sprite Size=31.3334
    UsePositionOffset=false
    Write Material Random=true
   )
   /SetVariables(Particles.Pos Choice:NiagaraInt32=&Multiply_Int(A=&RandomRangeInt(Minimum=7, Maximum=20), B=2)) {enabled=false, isSetParameters=true, module=SetVariables_#1}
   /SetVariables(Particles.Pos Choice:NiagaraInt32=&Multiply_Int(
     A=&SelectIntFromArray(
      Array Sampling Mode=::NewEnumerator1
      Direct Array Index=?
      Direct Array Mode=::NewEnumerator1
      Int Selection Array=DataInterface() {propertyValues {IntData [0,1,3,4,5,7,8,10,11,12]}}
      Random Seed=0
     )
     B=2
    )) {isSetParameters=true, module=SetVariables_#2}
   &InitialMeshOrientation(
    Mesh Orientation Mode=::NewEnumerator7
    Rotation=&RandomRangeVector(Minimum=<0 0 0>, Maximum=<1 0 0>)
    Use Rotation Vector=true
   )
   /SetVariables(Particles.Pos A Color:LinearColor=<rgba 1 0 0 1>, Particles.PosB COlor:LinearColor=<rgba 0 1 0 1>) {isSetParameters=true, module=SetVariables_#3}
  ]
  particleUpdateScript [
   &ParticleState()
   &self:Drift_Horror_Get_Set_of_4_Bones(New Skeletal Mesh:NiagaraDataInterfaceSkeletalMesh=$User.SKeletalMesh, StartIndex:NiagaraInt32=$Particles.Initial.Pos Choice) {module=Drift_Horror_Get_Set_of_4_Bones}
   /SetVariables(
    Particles.Pos A:Vector3f=&Lerp_Vector(
     A=$Output.Drift_Horror_Get_Set_of_4_Bones.BonePos1
     B=$Output.Drift_Horror_Get_Set_of_4_Bones.BonePos2
     Alpha=&RandomRangeFloat(Minimum=0, Maximum=1)
    )
    Particles.Pos B:Vector3f=&Lerp_Vector(
     A=$Output.Drift_Horror_Get_Set_of_4_Bones.BonePos3
     B=$Output.Drift_Horror_Get_Set_of_4_Bones.BonePos4
     Alpha=&RandomRangeFloat(Minimum=0, Maximum=1)
    )
   ) {isSetParameters=true, module=SetVariables_#4}
   /SetVariables(Particles.Position:Vector3f=$Particles.Pos A) {isSetParameters=true, module=SetVariables_#5}
   &DynamicMaterialParameters(
    Index 0 Param 1=1
    Index 0 Param 2=0.5
    Index 0 Param 3=-0.1
    Module.Param0WriteEnabled=true
    Module.Param1WriteEnabled=true
    Module.Param2WriteEnabled=true
    Module.Param3WriteEnabled=true
   )
   /Niagara/Modules/Update/Orientation/V2/OrientMeshToVector(
    Look At Direction:Vector3f=&NormalizeVector(
     Vector To Normalize=&Subtract_Vector(A=$Particles.Pos B, B=$Particles.Pos A)
     Safe Normalize=true
     Safe Fallback Threshold=0
     Safe Fallback Vector=<0 0 0.001>
    )
    Override Quat=false
    Up Axis Direction:Vector3f=<0 0 1>
    UseOptionalSideAxis=false
    UseOptionalUpAxis=false
   )
   &ScaleMeshSize(Scale Factor=&MakeVector(
     X=&VectorLength(Vector=&Subtract_Vector(A=$Particles.Pos B, B=$Particles.Pos A))
     Y=&FloatFromCurve(FloatCurve=DataInterface() {propertyValues {Curve {keys [<at 0 1>,<at 0.9029 2.7739>,<at 1 5>]}}})
     Z=&FloatFromCurve(FloatCurve=DataInterface() {propertyValues {Curve {keys [<at 0 1>,<at 0.9 2>,<at 1 3>]}}})
    ))
   &FN_Light_Attributes(
    Color=<rgba 25 6.25 20.4806 1>
    Create Color Attribute=true
    Create Exponent Attribute=true
    Create Position Attribute=true
    Position=&Lerp_Vector(A=$Particles.Pos A, B=$Particles.Pos B, Alpha=0.5)
    Scale Color=&FloatFromCurve(FloatCurve=DataInterface() {propertyValues {Curve {keys [
         {arriveTangent=5.7658, interpMode=RCIM_Cubic, leaveTangent=5.7658, tangentMode=RCTM_User, time=0, value=0}
         <at 0.3 0.9>
         {interpMode=RCIM_Cubic, tangentMode=RCTM_User, time=0.5, value=0.5}
         <at 0.8 0.8>
         {arriveTangent=-4.8787, interpMode=RCIM_Cubic, leaveTangent=-4.8787, tangentMode=RCTM_User, time=1, value=0}
        ]}}})
   )
   &ScaleColor(
    Scale Alpha=&Multiply_Float(
     A=&FloatFromCurve(FloatCurve=DataInterface() {propertyValues {Curve {keys [
          {tangentMode=RCTM_User, time=0, value=1}
          {arriveTangent=-0.0405, interpMode=RCIM_Cubic, leaveTangent=-0.0405, tangentMode=RCTM_User, time=0.7009, value=1}
          {arriveTangent=-8.0089, interpMode=RCIM_Cubic, leaveTangent=-8.0089, tangentMode=RCTM_User, time=1, value=0}
         ]}}})
     B=&Sine(
      Normalized Angle=$Emitter.Age
      Period=&RandomRangeFloat(Minimum=0.05, Maximum=0.2)
      Scale=2
      Bias=&RandomRangeFloat(Minimum=0.5, Maximum=3)
     )
    )
    ScaleA=true
    ScaleRGB=true
   )
  ]
  rendererClasses [&NiagaraMeshRendererProperties, &NiagaraLightRendererProperties]
  renderers [
   {rendererClass=&NiagaraMeshRendererProperties, rendererIndex=0}
   {rendererClass=&NiagaraLightRendererProperties, rendererIndex=1}
  ]
 }
 Lightning_Spike_Arcs001 {
  bEnabled=false
  emitterUpdateScript [
   &EmitterState(
    Life Cycle Mode=::NewEnumerator1
    Loop Duration=&RandomRangeFloat(Minimum=0, Maximum=1)
    Recalculate Duration Each Loop=true
    UseLoopDelay=false
   )
   &SpawnBurst_Instantaneous(
    Spawn Count=&MakeCustomIntFromBool(
     Bool=&SetBoolByFloatComparison(A=&RandomRangeFloat(Minimum=0, Maximum=1), B=0.25)
     True INT=1
     False INT=2
    )
    Use Spawn Probability=false
    UseLoopCountLimit=false
   )
   &SpawnRate(SpawnRate=500, Use Spawn Probability=false) {enabled=false}
  ]
  particleSpawnScript [
   &InitializeParticle(
    Color=<rgba 2 0 1.5333 1>
    Lifetime=0.25
    Mesh Scale Max=<0.037 1 1>
    Mesh Scale Min=<0.037 1.25 2>
    Mesh Scale Mode=::NewEnumerator2
    Position Mode=::NewEnumerator1
    Sprite Size Mode=::NewEnumerator3
    Uniform Sprite Size=31.3334
    UsePositionOffset=false
    Write Material Random=true
   )
   /SetVariables(Particles.Pos Choice:NiagaraInt32=&Multiply_Int(A=&RandomRangeInt(Minimum=7, Maximum=20), B=2)) {enabled=false, isSetParameters=true, module=SetVariables_#1}
   /SetVariables(Particles.Pos Choice:NiagaraInt32=&Multiply_Int(
     A=&SelectIntFromArray(
      Array Sampling Mode=::NewEnumerator1
      Direct Array Index=?
      Direct Array Mode=::NewEnumerator1
      Int Selection Array=DataInterface() {propertyValues {IntData [0,1,2,3,4,5,6,7,8,9,10,11,12,13,14,15,17,19,20]}}
      Random Seed=0
     )
     B=2
    )) {isSetParameters=true, module=SetVariables_#2}
   &InitialMeshOrientation(
    Mesh Orientation Mode=::NewEnumerator7
    Rotation=&RandomRangeVector(Minimum=<0 0 0>, Maximum=<1 0 0>)
    Use Rotation Vector=true
   )
   /SetVariables(Particles.Pos A Color:LinearColor=<rgba 1 0 0 1>, Particles.PosB COlor:LinearColor=<rgba 0 1 0 1>) {isSetParameters=true, module=SetVariables_#3}
  ]
  particleUpdateScript [
   &ParticleState()
   &self:Drift_Horror_Get_Set_of_4_Bones(New Skeletal Mesh:NiagaraDataInterfaceSkeletalMesh=$User.SKeletalMesh, StartIndex:NiagaraInt32=$Particles.Initial.Pos Choice) {module=Drift_Horror_Get_Set_of_4_Bones}
   /SetVariables(
    Particles.Pos A:Vector3f=&Lerp_Vector(
     A=$Output.Drift_Horror_Get_Set_of_4_Bones.BonePos1
     B=$Output.Drift_Horror_Get_Set_of_4_Bones.BonePos2
     Alpha=&RandomRangeFloat(Minimum=0, Maximum=1)
    )
    Particles.Pos B:Vector3f=&Lerp_Vector(
     A=$Output.Drift_Horror_Get_Set_of_4_Bones.BonePos3
     B=$Output.Drift_Horror_Get_Set_of_4_Bones.BonePos4
     Alpha=&RandomRangeFloat(Minimum=0, Maximum=1)
    )
   ) {enabled=false, isSetParameters=true, module=SetVariables_#4}
   /SetVariables(Particles.Position:Vector3f=$Output.Drift_Horror_Get_Set_of_4_Bones.BonePos1) {isSetParameters=true, module=SetVariables_#5}
   &DynamicMaterialParameters(
    Index 0 Param 1=1
    Index 0 Param 2=0.5
    Index 0 Param 3=-0.1
    Module.Param0WriteEnabled=true
    Module.Param1WriteEnabled=true
    Module.Param2WriteEnabled=true
    Module.Param3WriteEnabled=true
   )
   &OrientMeshToVector(
    Look At Direction:Vector3f=&NormalizeVector(
     Vector To Normalize=&Subtract_Vector(A=$Output.Drift_Horror_Get_Set_of_4_Bones.BonePos2, B=$Output.Drift_Horror_Get_Set_of_4_Bones.BonePos1)
     Safe Normalize=true
     Safe Fallback Threshold=0
     Safe Fallback Vector=<0 0 0.001>
    )
    Override Quat=false
    Up Axis Direction:Vector3f=<0 0 1>
    UseOptionalSideAxis=false
    UseOptionalUpAxis=false
   )
   &ScaleMeshSize(Scale Factor=&MakeVector(
     X=&VectorLength(Vector=&Multiply_Vector(
       A=&Multiply_VectorByFloat(
        Vector=&Subtract_Vector(A=$Output.Drift_Horror_Get_Set_of_4_Bones.BonePos2, B=$Output.Drift_Horror_Get_Set_of_4_Bones.BonePos1)
        Float=1
       )
       B=<1.5 1.5 1.5>
      ))
     Y=&FloatFromCurve(FloatCurve=DataInterface() {propertyValues {Curve {keys [<at 0 1>,<at 0.9029 2.7739>,<at 1 3.875>]}}})
     Z=1
    ))
   &FN_Light_Attributes(
    Color=<rgba 25 6.25 20.4806 1>
    Create Color Attribute=true
    Create Exponent Attribute=true
    Create Position Attribute=true
    Position=&Lerp_Vector(A=$Output.Drift_Horror_Get_Set_of_4_Bones.BonePos1, B=$Output.Drift_Horror_Get_Set_of_4_Bones.BonePos2, Alpha=0.5)
    Scale Color=&FloatFromCurve(FloatCurve=DataInterface() {propertyValues {Curve {keys [
         {arriveTangent=5.7658, interpMode=RCIM_Cubic, leaveTangent=5.7658, tangentMode=RCTM_User, time=0, value=0}
         <at 0.3 0.9>
         {interpMode=RCIM_Cubic, tangentMode=RCTM_User, time=0.5, value=0.5}
         <at 0.8 0.8>
         {arriveTangent=-4.8787, interpMode=RCIM_Cubic, leaveTangent=-4.8787, tangentMode=RCTM_User, time=1, value=0}
        ]}}})
   )
   &ScaleColor(
    Scale Alpha=&Multiply_Float(
     A=&FloatFromCurve(FloatCurve=DataInterface() {propertyValues {Curve {keys [
          {tangentMode=RCTM_User, time=0, value=1}
          {arriveTangent=-0.0405, interpMode=RCIM_Cubic, leaveTangent=-0.0405, tangentMode=RCTM_User, time=0.7009, value=1}
          {arriveTangent=-8.0089, interpMode=RCIM_Cubic, leaveTangent=-8.0089, tangentMode=RCTM_User, time=1, value=0}
         ]}}})
     B=&Sine(
      Normalized Angle=$Emitter.Age
      Period=&RandomRangeFloat(Minimum=0.05, Maximum=0.2)
      Scale=2
      Bias=&RandomRangeFloat(Minimum=0.5, Maximum=5)
     )
    )
    ScaleA=true
    ScaleRGB=true
   )
  ]
  rendererClasses [&NiagaraMeshRendererProperties, &NiagaraLightRendererProperties]
  renderers [
   {rendererClass=&NiagaraMeshRendererProperties, rendererIndex=0}
   {rendererClass=&NiagaraLightRendererProperties, rendererIndex=1}
  ]
 }
 Lightning_ou_of_spikes {
  emitterUpdateScript [
   &EmitterState(
    Life Cycle Mode=::NewEnumerator1
    Loop Duration=&RandomRangeFloat(Minimum=0, Maximum=0.55)
    Recalculate Duration Each Loop=true
    UseLoopDelay=false
   )
   &SpawnBurst_Instantaneous(
    Spawn Count=&MakeCustomIntFromBool(
     Bool=&SetBoolByFloatComparison(A=&RandomRangeFloat(Minimum=0, Maximum=1), B=0.25)
     True INT=1
     False INT=2
    )
    Use Spawn Probability=false
    UseLoopCountLimit=false
   )
   &SpawnRate(SpawnRate=5, Use Spawn Probability=false) {enabled=false}
  ]
  particleSpawnScript [
   &InitializeParticle(
    Color=<rgba 15 0 11.5 1>
    Lifetime Max=0.3
    Lifetime Min=0.2
    Lifetime Mode=::NewEnumerator1
    Mesh Scale Max=<2.5 2.5 2.5>
    Mesh Scale Min=<2 2 2>
    Mesh Scale Mode=::NewEnumerator2
    Position Mode=::NewEnumerator1
    Sprite Size Mode=::NewEnumerator3
    Uniform Sprite Size=2
    UsePositionOffset=false
    Write Material Random=true
   )
   /SetVariables(Particles.Pos Choice:NiagaraInt32=&Multiply_Int(A=&RandomRangeInt(Minimum=0, Maximum=20), B=2)) {isSetParameters=true, module=SetVariables_#1}
   &InitialMeshOrientation(
    Mesh Orientation Mode=::NewEnumerator7
    Rotation=&RandomRangeVector(Minimum=<0 0.25 0.25>, Maximum=<1 -0.25 -0.25>)
    Use Rotation Vector=true
   ) {enabled=false}
   &InitialMeshOrientation(
    Mesh Orientation Mode=::NewEnumerator2
    Orientation Coordinate Space=::Local
    Orientation Vector=<-0.25 -1 0>
    Rotation=&RandomRangeVector(Minimum=<0 0 0>, Maximum=<1 0 0>)
    Use Rotation Vector=true
   ) {module=InitialMeshOrientation001}
  ]
  particleUpdateScript [
   &ParticleState()
   &self:Drift_Horror_Get_Set_of_4_Bones(New Skeletal Mesh:NiagaraDataInterfaceSkeletalMesh=$User.SKeletalMesh, StartIndex:NiagaraInt32=$Particles.Initial.Pos Choice) {module=Drift_Horror_Get_Set_of_4_Bones}
   /SetVariables(
    Particles.Pos A:Vector3f=&Lerp_Vector(
     A=$Output.Drift_Horror_Get_Set_of_4_Bones.BonePos1
     B=$Output.Drift_Horror_Get_Set_of_4_Bones.BonePos2
     Alpha=&RandomRangeFloat(Minimum=0, Maximum=1)
    )
    Particles.Pos B:Vector3f=&Lerp_Vector(
     A=$Output.Drift_Horror_Get_Set_of_4_Bones.BonePos3
     B=$Output.Drift_Horror_Get_Set_of_4_Bones.BonePos4
     Alpha=&RandomRangeFloat(Minimum=0, Maximum=1)
    )
   ) {isSetParameters=true, module=SetVariables_#4}
   /SetVariables(Particles.Position:Vector3f=$Particles.Pos A) {isSetParameters=true, module=SetVariables_#5}
   &OrientMeshToVector(
    Look At Direction:Vector3f=&Subtract_Vector(A=$Output.Drift_Horror_Get_Set_of_4_Bones.BonePos2, B=$Output.Drift_Horror_Get_Set_of_4_Bones.BonePos1)
    Override Quat=false
    Up Axis Direction:Vector3f=<0 0 1>
    UseOptionalSideAxis=false
    UseOptionalUpAxis=false
   ) {enabled=false}
   &DynamicMaterialParameters(
    Index 0 Param 1=1
    Index 0 Param 3=-0.1
    Module.Param0WriteEnabled=true
    Module.Param1WriteEnabled=true
    Module.Param2WriteEnabled=true
    Module.Param3WriteEnabled=true
   )
   &ScaleMeshSize(Scale Factor=&MakeVector(
     X=&FloatFromCurve(FloatCurve=DataInterface() {propertyValues {Curve {keys [{tangentMode=RCTM_User, time=0, value=0.8}, {tangentMode=RCTM_User, time=0.2, value=1}, <at 0.9 0.85>, <at 1 1.25>]}}})
     Y=&Lerp_Float(A=1, B=&RandomRangeFloat(Minimum=0.5, Maximum=4), Alpha=$Particles.NormalizedAge)
     Z=&Lerp_Float(A=1, B=&RandomRangeFloat(Minimum=0.5, Maximum=4), Alpha=$Particles.NormalizedAge)
    ))
   &FN_Light_Attributes(
    Color=<rgba 5 1.25 4.0961 1>
    Create Color Attribute=true
    Create Exponent Attribute=true
    Create Position Attribute=true
    Position=&Add_Vector(
     A=$Particles.Pos A
     B=&Multiply_VectorByFloat(
      Vector=&Subtract_Vector(A=$Output.Drift_Horror_Get_Set_of_4_Bones.BonePos2, B=$Output.Drift_Horror_Get_Set_of_4_Bones.BonePos1)
      Float=1
     )
    )
   )
   &ScaleColor(
    Scale Alpha=&Sine(
     Normalized Angle=$Emitter.Age
     Period=&RandomRangeFloat(Minimum=0.05, Maximum=0.2)
     Scale=2
     Bias=&RandomRangeFloat(Minimum=0.5, Maximum=2)
    )
    ScaleA=true
    ScaleRGB=true
   )
  ]
  rendererClasses [&NiagaraMeshRendererProperties, &NiagaraLightRendererProperties]
  renderers [
   {rendererClass=&NiagaraMeshRendererProperties, rendererIndex=0}
   {rendererClass=&NiagaraLightRendererProperties, rendererIndex=1}
  ]
 }
 Lightning_ou_of_spikes_Edge {
  emitterUpdateScript [
   &EmitterState(
    Life Cycle Mode=::NewEnumerator1
    Loop Duration=&RandomRangeFloat(Minimum=0, Maximum=0.35)
    Recalculate Duration Each Loop=true
    UseLoopDelay=false
   )
   &SpawnBurst_Instantaneous(
    Spawn Count=&MakeCustomIntFromBool(
     Bool=&SetBoolByFloatComparison(A=&RandomRangeFloat(Minimum=0, Maximum=0.5), B=0.25)
     True INT=1
     False INT=2
    )
    Use Spawn Probability=false
    UseLoopCountLimit=false
   )
   &SpawnRate(SpawnRate=20, Use Spawn Probability=false) {enabled=false}
  ]
  particleSpawnScript [
   &InitializeParticle(
    Color=<rgba 5 0 3.8333 1>
    Lifetime Max=0.3
    Lifetime Min=0.2
    Lifetime Mode=::NewEnumerator1
    Mesh Scale Max=<3 3 3>
    Mesh Scale Min=<2 2 2>
    Mesh Scale Mode=::NewEnumerator2
    Position Mode=::NewEnumerator1
    Sprite Size Mode=::NewEnumerator3
    Uniform Sprite Size=2
    UsePositionOffset=false
    Write Material Random=true
   )
   /SetVariables(Particles.Pos Choice:NiagaraInt32=&Multiply_Int(
     A=&SelectIntFromArray(
      Array Sampling Mode=::NewEnumerator1
      Direct Array Index=?
      Int Selection Array=DataInterface() {propertyValues {IntData [0,6,7,13,14,40]}}
      Random Seed=0
     )
     B=2
    )) {isSetParameters=true, module=SetVariables_#1}
   &InitialMeshOrientation(Mesh Orientation Mode=::NewEnumerator7, Rotation=<0 0 -0.75>, Use Rotation Vector=true) {enabled=false}
   &InitialMeshOrientation(
    Mesh Orientation Mode=::NewEnumerator2
    Orientation Coordinate Space=::Local
    Orientation Vector=<-0.25 -1 0>
    Rotation=&RandomRangeVector(Minimum=<0 0 0>, Maximum=<1 0 0>)
    Use Rotation Vector=true
   ) {module=InitialMeshOrientation001}
  ]
  particleUpdateScript [
   &ParticleState()
   &self:Drift_Horror_Get_Set_of_4_Bones(New Skeletal Mesh:NiagaraDataInterfaceSkeletalMesh=$User.SKeletalMesh, StartIndex:NiagaraInt32=$Particles.Initial.Pos Choice) {module=Drift_Horror_Get_Set_of_4_Bones}
   /SetVariables(
    Particles.Pos A:Vector3f=&Lerp_Vector(
     A=$Output.Drift_Horror_Get_Set_of_4_Bones.BonePos1
     B=$Output.Drift_Horror_Get_Set_of_4_Bones.BonePos2
     Alpha=&RandomRangeFloat(Minimum=0, Maximum=1)
    )
    Particles.Pos B:Vector3f=&Lerp_Vector(
     A=$Output.Drift_Horror_Get_Set_of_4_Bones.BonePos3
     B=$Output.Drift_Horror_Get_Set_of_4_Bones.BonePos4
     Alpha=&RandomRangeFloat(Minimum=0, Maximum=1)
    )
   ) {isSetParameters=true, module=SetVariables_#4}
   /SetVariables(Particles.Position:Vector3f=$Particles.Pos A) {isSetParameters=true, module=SetVariables_#5}
   &OrientMeshToVector(
    Look At Direction:Vector3f=&Subtract_Vector(A=$Particles.Pos A, B=/Niagara/DynamicInputs/Helpers/SimulationPosition())
    Override Quat=false
    Up Axis Direction:Vector3f=<0 0 1>
    UseOptionalSideAxis=true
    UseOptionalUpAxis=true
   ) {enabled=false}
   &ScaleMeshSize(Scale Factor=&MakeVector(
     X=&FloatFromCurve(FloatCurve=DataInterface() {propertyValues {Curve {keys [{tangentMode=RCTM_User, time=0, value=0.8}, {tangentMode=RCTM_User, time=0.2, value=1}, <at 0.9 0.85>, <at 1 1.25>]}}})
     Y=&Lerp_Float(A=1, B=&RandomRangeFloat(Minimum=0.5, Maximum=2), Alpha=$Particles.NormalizedAge)
     Z=&Lerp_Float(A=1, B=&RandomRangeFloat(Minimum=0.5, Maximum=2), Alpha=$Particles.NormalizedAge)
    ))
   &DynamicMaterialParameters(
    Index 0 Param 1=1
    Index 0 Param 2=50
    Index 0 Param 3=-0.1
    Index 0 Param 4=0.5
    Module.Param0WriteEnabled=true
    Module.Param1WriteEnabled=true
    Module.Param2WriteEnabled=true
    Module.Param3WriteEnabled=true
   )
   &FN_Light_Attributes(
    Color=<rgba 5 1.25 4.0961 1>
    Create Color Attribute=true
    Create Exponent Attribute=true
    Create Position Attribute=true
    Position=&Add_Vector(
     A=$Particles.Pos A
     B=&Multiply_VectorByFloat(
      Vector=&Subtract_Vector(A=$Output.Drift_Horror_Get_Set_of_4_Bones.BonePos2, B=$Output.Drift_Horror_Get_Set_of_4_Bones.BonePos1)
      Float=1
     )
    )
   )
   &ScaleColor(
    Scale Alpha=&Sine(
     Normalized Angle=$Emitter.Age
     Period=&RandomRangeFloat(Minimum=0.05, Maximum=0.2)
     Scale=2
     Bias=&RandomRangeFloat(Minimum=0.5, Maximum=2)
    )
    ScaleA=true
    ScaleRGB=true
   )
  ]
  rendererClasses [&NiagaraMeshRendererProperties, &NiagaraLightRendererProperties]
  renderers [
   {rendererClass=&NiagaraMeshRendererProperties, rendererIndex=0}
   {rendererClass=&NiagaraLightRendererProperties, rendererIndex=1}
  ]
 }
 Ribbon_Tails {
  emitterUpdateScript [
   &EmitterState(Inactive Response=::NewEnumerator1, Life Cycle Mode=::NewEnumerator1, Loop Duration=2, UseLoopDelay=false)
   &SpawnRate(SpawnRate=15, Use Spawn Probability=false)
  ]
  particleSpawnScript [
   &InitializeRibbon(
    Color=<rgba 0.0108 0.0025 0.025 1>
    Lifetime=1.25
    Ribbon Width=&RandomRangeFloat(Minimum=25, Maximum=12)
    UsePositionOffset=false
    Write Material Random=true
   )
   &FN_SetSpawnGroup(Number of Spawn Groups=2, Set Ribbon ID and Ribbon Link Order=true)
   /SetVariables(
    Particles.RibbonU0Override=&Add_Float(
     A=&Add_Float(
      A=&Multiply_Float(A=$Emitter.Age, B=0.75)
      B=&MaskFloatBySpawnGroup(Masked Float=0, Passthrough Float=0.5, Spawn Group Mask=1)
     )
     B=&RandomRangeFloat(Minimum=-0.05, Maximum=0.05)
    )
    Particles.SpriteSize=<5 5>
   ) {isSetParameters=true}
  ]
  particleUpdateScript [
   &ParticleState()
   /SetVariables(Particles.RibbonU0Override=&Add_Float(A=$Particles.Initial.RibbonU0Override, B=&Multiply_Float(A=$Emitter.Age, B=0.15))) {isSetParameters=true}
   /SetVariables(Particles.Owner Vel:Vector3f=&Multiply_VectorByFloat(
     Vector=&NormalizeVector(
      Vector To Normalize=&Multiply_Vector(
       A=&TransformVector(Vector=$Engine.Owner.Velocity, Source Space=::World, Destination Space=::Local)
       B=<1 1 1>
      )
      Safe Normalize=true
      Safe Fallback Threshold=0
      Safe Fallback Vector=<0 0 0.001>
     )
     Float=100
    )) {isSetParameters=true}
   &self:Drift_Horror_Get_Skelton_pos(
    New Skeletal Mesh:NiagaraDataInterfaceSkeletalMesh=$User.SkeletalMesh
    StartIndex:NiagaraInt32=&SelectIntFromArray(Direct Array Index=&SpawnGroup(), Int Selection Array=DataInterface() {propertyValues {IntData [0,4]}})
   ) {module=Drift_Horror_Get_Skelton_pos}
   &InterpolateOverTime(Target Vector Value:Vector3f=$Output.Drift_Horror_Get_Skelton_pos.BonePos2, Variable Type=::NewEnumerator2)
   &InterpolateOverTime(Target Vector Value:Vector3f=$Output.Drift_Horror_Get_Skelton_pos.BonePos3, Variable Type=::NewEnumerator2) {module=InterpolateOverTime002}
   &InterpolateOverTime(Target Vector Value:Vector3f=$Output.Drift_Horror_Get_Skelton_pos.BonePos4, Variable Type=::NewEnumerator2) {module=InterpolateOverTime001}
   /Game/Effects/Niagara/Modules/Curves/FN_HermiteSpline(
    Direction End:Vector3f=$Particles.InterpolateOverTime002.InterpolateOverTime_V3_.Moving Average
    Direction Start:Vector3f=$Particles.InterpolateOverTime.InterpolateOverTime_V3_.Moving Average
    Position Blend:NiagaraFloat=&FloatFromCurve(FloatCurve=DataInterface() {propertyValues {Curve {keys [{tangentMode=RCTM_User, time=0.1, value=0}, {tangentMode=RCTM_User, time=1, value=1}]}}})
    Position End:Vector3f=&Add_Vector(
     A=$Particles.InterpolateOverTime001.InterpolateOverTime_V3_.Moving Average
     B=&Lerp_Vector(A=<0 0 0>, B=<0 0 -250>, Alpha=$System.InterpolateOverTime.InterpolateOverTime_Float.Moving Average)
    )
    Position Start:Vector3f=$Output.Drift_Horror_Get_Skelton_pos.BonePos1
    Write Particles Position=true
   ) {module=FN_HermiteSpline001}
   &ScaleColor(
    Scale Alpha=&Multiply_Float(
     A=&Multiply_Float(
      A=&FloatFromCurve(FloatCurve=DataInterface() {propertyValues {Curve {keys [
           <at -0.1 0 RCIM_Cubic>
           {arriveTangent=58.3056, interpMode=RCIM_Cubic, leaveTangent=-3.206, tangentMode=RCTM_Break, time=0.25, value=5}
           {arriveTangent=-17.8036, interpMode=RCIM_Cubic, leaveTangent=-17.8036, tangentMode=RCTM_User, time=0.75, value=0}
          ]}}})
      B=&FloatFromCurve(
       FloatCurve=DataInterface() {propertyValues {Curve {keys [{arriveTangent=2.7248, interpMode=RCIM_Cubic, leaveTangent=2.7248, tangentMode=RCTM_User, time=0, value=0}, <at 0.9 1>, <at 1 0>]}}}
       CurveIndex=&ReturnNormalizedExecIndex()
      )
     )
     B=$System.Timeline002.NormalizedTime
    )
    ScaleA=true
    ScaleRGB=true
   )
   &ScaleRibbonWidth(Ribbon Width Scale=&FloatFromCurve(FloatCurve=DataInterface() {propertyValues {Curve {keys [<at 0 1>,<at 1 0.5>]}}}, Scale Curve=2))
   /SetVariables(Particles.RibbonLinkOrder:NiagaraFloat=$Particles.NormalizedAge) {isSetParameters=true}
   /SetVariables(Particles.Wiggled Position:Vector3f=&Add_Vector(
     A=Linked() {linkedVariable {name=Particles.Position, type=NiagaraPosition}}
     B=&Multiply_VectorByFloat(Vector=&RandomVector(Vector Scale=25), Float=$Particles.NormalizedAge)
    )) {isSetParameters=true}
  ]
  rendererClasses [&NiagaraRibbonRendererProperties]
  renderers [{rendererClass=&NiagaraRibbonRendererProperties}]
 }
 SmokePuff_Gen5 {
  emitterUpdateScript [
   &EmitterState()
   &SpawnBurst_Instantaneous(Spawn Count=5, Use Spawn Probability=false)
   &SpawnRate(SpawnRate=2, Use Spawn Probability=false)
   &SpawnRate(SpawnRate=90, Use Spawn Probability=false) {enabled=false, module=SpawnRate001}
   &SpawnBurst_Instantaneous(Spawn Count=5, Spawn Time=5.1315, Use Spawn Probability=false, UseLoopCountLimit=false) {module=SpawnBurst_Instantaneous001}
  ]
  particleSpawnScript [
   &InitializeParticle(
    Color Maximum=<rgba 0.0215 0.005 0.05 0.5>
    Color Minimum=<rgba 0.0323 0.0075 0.075 0.75>
    Color Mode=::NewEnumerator2
    Lifetime Max=4
    Lifetime Min=3
    Lifetime Mode=::NewEnumerator1
    Sprite Rotation Mode=::NewEnumerator2
    Sprite Size Mode=::NewEnumerator1
    Sprite UV Mode=::NewEnumerator3
    Uniform Sprite Size Max=150
    Uniform Sprite Size Min=100
    UsePositionOffset=false
    Write Material Random=true
   ) {module=InitializeParticle001}
   &CameraOffset(Camera Offset Amount=&RandomRangeFloat(Minimum=0, Maximum=25, Evaluation Type=::NewEnumerator1))
   &AddVelocity(Velocity=&Multiply_VectorByFloat(Vector=$Output.SphereLocation.SphereVector, Float=10)) {enabled=false}
   &CylinderLocation(
    Cylinder Axis:Vector3f=<0 1 -0.1>
    Cylinder Height=2
    Cylinder Radius=12
    MaskSpawnGroup=false
    Non Uniform Scale:Vector3f=<1.5 0.5 1>
    Override Seed=false
    Rotation Mode:ENiagara_RotationMode=::NewEnumerator1
    UseOwnerScale=true
   ) {enabled=false}
   &InheritVelocity(Inherited Velocity Amount Scale=<0.25 0.25 0.25>, Speed Limit=true, UseSpeedThreshold=false) {enabled=false}
  ]
  particleUpdateScript [
   &ParticleState()
   &ScaleSpriteSize(Scale Factor=&Vector2DFromFloat(Value=&FloatFromCurve(
      FloatCurve=DataInterface() {propertyValues {Curve {keys [
          {arriveTangent=3.2112, interpMode=RCIM_Cubic, leaveTangent=3.2112, tangentMode=RCTM_User, time=0, value=0.1}
          {
           arriveTangentWeight=0.6453
           interpMode=RCIM_Cubic
           leaveTangentWeight=0.6453
           tangentMode=RCTM_User
           tangentWeightMode=RCTWM_WeightedBoth
           time=1
           value=1
          }
         ]}}}
      Scale Curve=0.5
     )))
   &ScaleColor(
    Scale Alpha=&FloatFromCurve(
     FloatCurve=DataInterface() {propertyValues {Curve {keys [
         <at 0 0 RCIM_Cubic>
         {interpMode=RCIM_Cubic, leaveTangent=-2.3157, tangentMode=RCTM_Break, time=0.15, value=1}
         {interpMode=RCIM_Cubic, leaveTangent=-2.2979, time=1, value=0}
        ]}}}
     Scale Curve=8
    )
    ScaleA=true
    ScaleRGB=false
   )
   &SkeletalMeshLocation(
    Bone / Socket Index=&RandomRangeInt(Minimum=8, Maximum=9)
    MaskSpawnGroup=false
    Mesh Sampling Type=::NewEnumerator5
    Orientation Sampling=::NewEnumerator1
    Sampled Position Offset=&Add_Vector(
     A=&RandomRangeVector(Minimum=<-5 20 -5>, Maximum=<5 -25 5>)
     B=&Lerp_Vector(A=<0 0 0>, B=<0 20 0>, Alpha=$Particles.NormalizedAge)
    )
    Skeletal Mesh=$User.SkeletalMesh
    Socket Sampling Mode=::NewEnumerator5
   )
   &GravityForce(Gravity=<0 0 -10>) {enabled=false}
   &DynamicMaterialParameters(
    Index 0 Param 1=&FloatFromCurve(
     FloatCurve=DataInterface() {propertyValues {Curve {keys [
         <at 0 0 RCIM_Cubic>
         {arriveTangent=1.5947, interpMode=RCIM_Cubic, leaveTangent=1.5947, tangentMode=RCTM_User, time=0.2557, value=0.5257}
         <at 1 1 RCIM_Cubic>
        ]}}}
     CurveIndex=$Particles.Age
     Scale Curve=0.25
    )
    Index 0 Param 4=1
    Module.Param0WriteEnabled=true
    Module.Param1WriteEnabled=true
    Module.Param2WriteEnabled=true
    Module.Param3WriteEnabled=true
   ) {module=DynamicMaterialParameters001}
   &SolveForcesAndVelocity(Clamp Velocity=false, Limit Acceleration=false) {enabled=false}
   &SubUVAnimation(SubUV Loop Count=1.25)
  ]
  rendererClasses [&NiagaraSpriteRendererProperties]
  renderers [
   {rendererClass=&NiagaraSpriteRendererProperties}
   {rendererClass=&NiagaraSpriteRendererProperties, rendererIndex=1}
  ]
 }
 Tail {
  bEnabled=false
  emitterUpdateScript [&EmitterState()]
  particleSpawnScript [&InitializeRibbon(Color=<rgba 0 0 0 25>, Lifetime=0.1, UsePositionOffset=false, Write Material Random=true)]
  particleUpdateScript [
   &ParticleState()
   &InheritSourceMovement(ApplyForce=false, ApplyPosition=true, ApplyPositionFalloff=false)
   &ScaleColor(
    Scale Alpha=&FloatFromCurve(FloatCurve=DataInterface() {propertyValues {Curve {keys [<at 0 1>,<at 1 0>]}}})
    ScaleA=true
    ScaleRGB=true
   )
   &CurlNoiseForce(Noise Strength=&Multiply_Float(A=500, B=$Particles.NormalizedAge), Pan Noise=false) {enabled=false}
   &CurlNoiseForce(Noise Strength=100, Pan Noise=false) {module=CurlNoiseForce001}
   &SolveForcesAndVelocity(Clamp Velocity=false, Limit Acceleration=false)
   &ScaleRibbonWidth(Ribbon Width Scale=&FloatFromCurve(FloatCurve=DataInterface() {propertyValues {Curve {keys [<at 0 1>, {leaveTangent=-7.8977, time=0.8, value=0}]}}}))
  ]
  rendererClasses [&NiagaraSpriteRendererProperties, &NiagaraRibbonRendererProperties]
  renderers [
   {rendererClass=&NiagaraSpriteRendererProperties}
   {rendererClass=&NiagaraRibbonRendererProperties, rendererIndex=1}
  ]
 }
 pendulum_Drip {
  emitterUpdateScript [
   &EmitterState(
    Life Cycle Mode=::NewEnumerator1
    Loop Duration=&RandomRangeFloat(Minimum=4, Maximum=8, Evaluation Type=::NewEnumerator1)
    UseLoopDelay=false
   )
   &SpawnBurst_Instantaneous(Spawn Count=1, Use Spawn Probability=false)
   &SpawnRate(SpawnRate=3, Use Spawn Probability=false)
  ]
  particleSpawnScript [
   /SetVariables(Particles.PendulumDrag=5, Particles.PendulumLengthScale=0.2, Particles.PendulumMaskFalloff=1.5, Particles.PendulumRigidity=0.2) {isSetParameters=true}
   /Niagara/Modules/Spawn/Initialization/InitializeParticle(
    Color=&FN_Select_Color_from_Time_Of_Day(Morning=<rgba 1 1 1 1>, Day=<rgba 0.6 0.6 0.6 1>, Evening=<rgba 0.3 0.3 0.3 1>, Night=<rgba 0.125 0.125 0.125 1>)
    Lifetime=&UniformRangedFloat(Minimum=1, Maximum=1.5)
    Mesh Scale=&Multiply_Vector(A=<1 1 1>, B=&MakeVector(X=0.1, Y=0.1, Z=$Particles.PendulumLengthScale))
    Module.Write Color=true
    Module.Write Lifetime=true
    Module.Write Mass=true
    Module.Write Material Random=true
    Module.Write Position=false
    Module.Write Scale=true
    Module.Write SpriteRotation=false
    Module.Write SpriteSize=true
   )
   /SetVariables(Particles.RandomOffset:Vector3f=/Niagara/DynamicInputs/Random/RandomVector(
     Vector Scale:NiagaraFloat=&UniformRangedFloat(Minimum=25, Maximum=202.999)
     Constraint Vector:Vector3f=<0 0 1>
     Randomness Mode:ENiagaraRandomnessMode=::NewEnumerator0
     Random Seed:NiagaraInt32=0
     Fixed Random Seed:NiagaraBool=false
    )) {enabled=false, isSetParameters=true}
   &CameraOffset()
   /SetVariables(Particles.Skel Location Random:NiagaraInt32=&RandomRangeInt(Minimum=0, Maximum=108)) {isSetParameters=true}
  ]
  particleUpdateScript [
   &SkeletalMeshLocation(
    Apply To Particle Position=false
    Bone / Socket Index=$Particles.Skel Location Random
    Bone Sampling Mode=::NewEnumerator5
    Inherit Velocity=false
    MaskSpawnGroup=false
    Skeletal Mesh=$User.SkeletalMesh
   )
   &ParticleState()
   /SetVariables(Particles.MeshOrigin:Vector3f=&Add_Vector(
     A=&Add_Vector(
      A=$Output.SkeletalMeshLocation.SampledPosition
      B=&MakeVector(
       X=&Multiply_Float(
        A=&MakeFloatFromVector(Vector=$Engine.Owner.Velocity, Channel=::NewEnumerator0)
        B=&FloatFromCurve(
         FloatCurve=DataInterface() {propertyValues {Curve {keys [
             <at 0 0>
             {arriveTangent=1.5112, interpMode=RCIM_Cubic, leaveTangent=1.5112, tangentMode=RCTM_User, time=0.7, value=0}
             {arriveTangent=4.7214, interpMode=RCIM_Cubic, leaveTangent=4.7214, tangentMode=RCTM_User, time=1, value=1}
            ]}}}
         Scale Curve=-0.075
        )
       )
       Y=&Multiply_Float(
        A=&MakeFloatFromVector(Vector=$Engine.Owner.Velocity, Channel=::NewEnumerator1)
        B=&FloatFromCurve(
         FloatCurve=DataInterface() {propertyValues {Curve {keys [
             <at 0 0>
             {arriveTangent=1.5112, interpMode=RCIM_Cubic, leaveTangent=1.5112, tangentMode=RCTM_User, time=0.7, value=0}
             {arriveTangent=4.7214, interpMode=RCIM_Cubic, leaveTangent=4.7214, tangentMode=RCTM_User, time=1, value=1}
            ]}}}
         Scale Curve=-0.075
        )
       )
       Z=&FloatFromCurve(
        FloatCurve=DataInterface() {propertyValues {Curve {keys [
            <at 0 0>
            {arriveTangent=1.5112, interpMode=RCIM_Cubic, leaveTangent=1.5112, tangentMode=RCTM_User, time=0.7, value=0}
            {arriveTangent=4.7214, interpMode=RCIM_Cubic, leaveTangent=4.7214, tangentMode=RCTM_User, time=1, value=1}
           ]}}}
        Scale Curve=-170
       )
      )
     )
     B=<0 0 0>
    )) {isSetParameters=true}
   &Drag(Drag=$Particles.PendulumDrag, Use Linear Drag=true, Use Rotational Drag=false)
   &GravityForce(Gravity=&Lerp_Vector(
     A=<0 0 -5000>
     B=<0 0 -12000>
     Alpha=&FloatFromCurve(FloatCurve=DataInterface() {propertyValues {Curve {keys [
          <at 0 0>
          {
           arriveTangent=9.4115
           arriveTangentWeight=2.2084
           interpMode=RCIM_Cubic
           leaveTangent=17.7358
           leaveTangentWeight=0.9539
           tangentMode=RCTM_Break
           tangentWeightMode=RCTWM_WeightedBoth
           time=0.6482
           value=0
          }
          {
           arriveTangent=0.0888
           arriveTangentWeight=0.3328
           interpMode=RCIM_Cubic
           leaveTangent=2.2069
           tangentMode=RCTM_Break
           tangentWeightMode=RCTWM_WeightedBoth
           time=1
           value=1
          }
         ]}}})
    ))
   &PendulumSetup(
    Module.Enable Spring Driver=true
    Pendulum Length:NiagaraFloat=&Multiply_Float(A=100, B=$Particles.PendulumLengthScale)
    Pendulum Pivot:Vector3f=$Particles.MeshOrigin
    "Spring Driven Constraint (Tightness)"=5
   )
   &InheritVelocity(Inherited Velocity Amount Scale=<50 50 50>, Inherited Velocity Speed Limit=100000, Speed Limit=true, UseSpeedThreshold=false) {enabled=false}
   &InheritSourceMovement(ApplyForce=false, ApplyPosition=true, ApplyPositionFalloff=false) {enabled=false}
   &SolveForcesAndVelocity(Clamp Velocity=false, Limit Acceleration=false)
   &PendulumConstraint()
   &CalculateAccurateVelocity()
   /SetVariables(Particles.MeshFacing:Vector3f=&NormalizeVector(Vector To Normalize=&Subtract_Vector(A=$Particles.MeshOrigin, B=Linked() {linkedVariable {name=Particles.Position, type=NiagaraPosition}}))) {isSetParameters=true}
   &OrientMeshToVector(
    Look At Direction:Vector3f=&Multiply_VectorByFloat(Vector=$Particles.MeshFacing, Float=1)
    Override Quat=false
    Reference Axis Vector:Vector3f=<0 0 1>
    UseOptionalSideAxis=false
    UseOptionalUpAxis=false
   )
   &DynamicMaterialParameters(
    Index 2 Param 1=HlslExpression() {hlslExpression=Particles.MeshOrigin.x}
    Index 2 Param 2=HlslExpression() {hlslExpression=Particles.MeshOrigin.y}
    Index 2 Param 3=HlslExpression() {hlslExpression=Particles.MeshOrigin.z}
    Index 2 Param 4=$Particles.PendulumMaskFalloff
    Index 3 Param 1=&MakeFloatFromVector(Vector=Linked() {linkedVariable {name=Particles.Position, type=NiagaraPosition}}, Channel=::NewEnumerator0)
    Index 3 Param 2=&MakeFloatFromVector(Vector=Linked() {linkedVariable {name=Particles.Position, type=NiagaraPosition}}, Channel=::NewEnumerator1)
    Index 3 Param 3=&MakeFloatFromVector(Vector=Linked() {linkedVariable {name=Particles.Position, type=NiagaraPosition}}, Channel=::NewEnumerator2)
    Index 3 Param 4=$Particles.PendulumRigidity
    Module.Param0WriteEnabled2=true
    Module.Param0WriteEnabled3=true
    Module.Param1WriteEnabled2=true
    Module.Param1WriteEnabled3=true
    Module.Param2WriteEnabled2=true
    Module.Param2WriteEnabled3=true
    Module.Param3WriteEnabled2=true
    Module.Param3WriteEnabled3=true
    Write Parameter Index 0=false
    Write Parameter Index 2=true
    Write Parameter Index 3=true
   )
   &ScaleMeshSize(Scale Factor=&MakeVector(
     X=&FloatFromCurve(FloatCurve=DataInterface() {propertyValues {Curve {keys [<at 0 0.2 RCIM_Cubic>,<at 0.7 1 RCIM_Cubic>]}}})
     Y=&FloatFromCurve(FloatCurve=DataInterface() {propertyValues {Curve {keys [<at 0 0.2 RCIM_Cubic>,<at 0.7 1 RCIM_Cubic>]}}})
     Z=&FloatFromCurve(FloatCurve=DataInterface() {propertyValues {Curve {keys [
          <at 0 0.25 RCIM_Cubic>
          <at 0.7 0.5 RCIM_Cubic>
          {arriveTangent=-1.5491, interpMode=RCIM_Cubic, leaveTangent=-1.5491, tangentMode=RCTM_User, time=1, value=0.25}
         ]}}})
    ))
  ]
  rendererClasses [&NiagaraMeshRendererProperties, &NiagaraSpriteRendererProperties]
  renderers [
   {rendererClass=&NiagaraMeshRendererProperties, rendererIndex=0}
   {rendererClass=&NiagaraSpriteRendererProperties, rendererIndex=1}
  ]
 }
}
summary {
 systemName=NS_DriftHorror_Glider
 emitters [
  {emitterName=Ribbon_Tails, rendererClasses [&NiagaraRibbonRendererProperties]}
  {
   emitterName=Lightning_Spike_Arcs
   rendererClasses [&NiagaraMeshRendererProperties, &NiagaraLightRendererProperties]
  }
  {
   emitterName=Lightning_ou_of_spikes
   rendererClasses [&NiagaraMeshRendererProperties, &NiagaraLightRendererProperties]
  }
  {
   emitterName=Lightning_ou_of_spikes_Edge
   rendererClasses [&NiagaraMeshRendererProperties, &NiagaraLightRendererProperties]
  }
  {emitterName=SmokePuff_Gen5, rendererClasses [&NiagaraSpriteRendererProperties]}
  {bEnabled=false, emitterName=Drips_Falling, rendererClasses [&NiagaraSpriteRendererProperties]}
  {
   bEnabled=false
   emitterName=Lightning_Spike_Arcs001
   rendererClasses [&NiagaraMeshRendererProperties, &NiagaraLightRendererProperties]
  }
  {
   bEnabled=false
   emitterName=Tail
   rendererClasses [&NiagaraSpriteRendererProperties, &NiagaraRibbonRendererProperties]
  }
  {emitterName=Bone_Flow_Ribbon02, rendererClasses [&NiagaraRibbonRendererProperties]}
  {emitterName=pendulum_Drip, rendererClasses [&NiagaraMeshRendererProperties, &NiagaraSpriteRendererProperties]}
 ]
 userVariables [
  User.bIsFrontEndPreview {type=NiagaraBool}
  User.ForwardDot {type=NiagaraFloat, defaultValue=0, description=""}
  User.RibbonIDOffset {type=NiagaraInt32}
  User.RightDot {type=NiagaraFloat, defaultValue=0, description=""}
  User.SkeletalMesh {
   type=NiagaraDataInterfaceSkeletalMesh
   defaultValue {
    struct=NiagaraExt_VariableValue_DataInterface
    value {dataInterface=&self:NiagaraDataInterfaceSkeletalMesh_0}
   }
  }
 ]
}
