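# Niagara deploy effect for the Drift Horror male glider, combining a splashy burst with magenta lightning.
# Droplets inherit owner velocity alongside a shockwave and streaks; tailed drops emit location events, apparently feeding the Tail ribbons.
# Arc meshes read bone positions from the user skeletal mesh, then orient and stretch between two points.
# Spike meshes flicker through sine-driven alpha, with particle lights on the hero spikes and arcs.
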NiagaraSystem /BRCosmetics/Gadgets/Assets/VinderTech_GliderChute/Glider_Drift_Horror_Male/FX/NS_DriftHorror_Glider_Deply
dependencies {
 usedDataInterfaces [
  &NiagaraDataInterfaceCurve
  &NiagaraDataInterfaceVector2DCurve
  &NiagaraDataInterfaceArrayFloat3
  &NiagaraDataInterfaceArrayInt32
 ]
 usedDynamicInputs []
 usedModules [&SystemState]
}
emitters {
 Drops {
  emitterUpdateScript [
   &EmitterState()
   &SpawnBurst_Instantaneous(Spawn Count=&RandomRangeInt(Minimum=6, Maximum=8), Use Spawn Probability=false)
   &SpawnRate(SpawnRate=250, Use Spawn Probability=false) {enabled=false}
   &SpawnPerUnit(Module.Use Max Movement Threshold=true, Module.Use Movement Tolerance=false, Use Spawn Probability=false)
  ]
  particleSpawnScript [
   &InitializeParticle(
    Lifetime=&RandomRangeFloat(Minimum=0.4, Maximum=0.6)
    Position=$Engine.Owner.Position
    Position Mode=::NewEnumerator1
    Sprite Size=&Multiply_Vector2(A=&RandomRangeVector2D(Minimum=<2 4>, Maximum=<4 6>), B=<1.25 5>)
    Sprite Size Mode=::NewEnumerator4
    UsePositionOffset=false
    Write Material Random=true
   )
   &AddVelocity(Velocity=&Multiply_VectorByFloat(Vector=&RandomRangeVector(Minimum=<-1 -1 -1>, Maximum=<1 1 1.5>), Float=100))
   &SphereLocation(
    MaskSpawnGroup=false
    Override Seed=false
    Sphere Radius=10
    Surface Only=false
    UseNonUniformScale=false
    UseOffset=true
    UseSphereOrientation=false
   )
   &AddVelocityFromPoint(
    Module.Use Falloff Distance=false
    Origin Offset Coordinate Space=::World
    Velocity Strength=&Multiply_Float(A=&RandomRangeFloat(Minimum=360, Maximum=440), B=1.25)
   )
   &CameraOffset(Camera Offset Amount=&RandomRangeFloat(Minimum=2, Maximum=5))
   &AddVelocityFromPoint(
    Module.Use Falloff Distance=false
    Origin Offset=<0 0 -0.1>
    Origin Offset Coordinate Space=::World
    Velocity Origin:NiagaraPosition=$Output.SphereLocation.Position
    Velocity Strength=&RandomRangeFloat(Minimum=150, Maximum=200)
   ) {enabled=false, module=AddVelocityFromPoint001}
   &InheritVelocity(Inherited Velocity Amount Scale=<5000 5000 5000>, Inherited Velocity Speed Limit=50000000188416, Speed Limit=true, UseSpeedThreshold=false) {enabled=false}
   &AddVelocity(Coordinate Space=::World, Velocity=&Multiply_VectorByFloat(Vector=$Engine.Owner.Velocity, Float=1000)) {module=AddVelocity001}
  ]
  particleUpdateScript [
   &InheritSourceMovement(ApplyForce=false, ApplyPosition=true, ApplyPositionFalloff=false)
   &GravityForce(Gravity=<0 0 -245>)
   &Drag(Drag=3, Use Linear Drag=true, Use Rotational Drag=false)
   &SolveForcesAndVelocity(Clamp Velocity=false, Limit Acceleration=false)
   &ParticleState()
   &ScaleColor(
    Scale Alpha=&FloatFromCurve(FloatCurve=DataInterface() {propertyValues {Curve {keys [
         {
          arriveTangentWeight=0.9254
          interpMode=RCIM_Cubic
          leaveTangentWeight=0.9254
          tangentMode=RCTM_User
          tangentWeightMode=RCTWM_WeightedBoth
          time=0
          value=1
         }
         {
          arriveTangent=-7.1278
          arriveTangentWeight=1.1179
          interpMode=RCIM_Cubic
          leaveTangent=-7.1278
          leaveTangentWeight=1.1179
          tangentMode=RCTM_User
          tangentWeightMode=RCTWM_WeightedBoth
          time=1
          value=0
         }
        ]}}})
    ScaleA=true
    ScaleRGB=true
   )
   &ScaleSpriteSize(
    Non-Uniform Curve Sprite Scale=DataInterface() {propertyValues {
      XCurve {keys [{tangentMode=RCTM_User, time=0, value=1}, {tangentMode=RCTM_User, time=1, value=0.75}]}
      YCurve {keys [{tangentMode=RCTM_User, time=0, value=1}, {tangentMode=RCTM_User, time=1, value=0.25}]}
     }}
    Scale Sprite Size Mode=::NewEnumerator3
   )
   /SetVariables(Particles.Position=&Add_Vector(
     A=Linked() {linkedVariable {name=Particles.Position, type=NiagaraPosition}}
     B=Linked() {linkedVariable {name=Engine.Owner.Position, type=NiagaraPosition}}
    )) {enabled=false, isSetParameters=true, module=SetVariables_#1}
  ]
  rendererClasses [&NiagaraSpriteRendererProperties, &NiagaraMeshRendererProperties]
  renderers [
   {rendererClass=&NiagaraSpriteRendererProperties}
   {rendererClass=&NiagaraMeshRendererProperties, rendererIndex=1}
  ]
 }
 Drops001 {
  emitterUpdateScript [
   &EmitterState()
   &SpawnBurst_Instantaneous(Spawn Count=&RandomRangeInt(Minimum=6, Maximum=8), Use Spawn Probability=false)
   &SpawnRate(SpawnRate=250, Use Spawn Probability=false) {enabled=false}
   &SpawnPerUnit(Module.Use Max Movement Threshold=true, Module.Use Movement Tolerance=false, Use Spawn Probability=false)
  ]
  particleSpawnScript [
   &InitializeParticle(
    Lifetime=&RandomRangeFloat(Minimum=0.4, Maximum=0.6)
    Position=$Engine.Owner.Position
    Position Mode=::NewEnumerator1
    Sprite Size=&Multiply_Vector2(A=&RandomRangeVector2D(Minimum=<2 4>, Maximum=<4 6>), B=<1.5 2>)
    Sprite Size Mode=::NewEnumerator4
    UsePositionOffset=false
    Write Material Random=true
   )
   &AddVelocity(Velocity=&Multiply_VectorByFloat(Vector=&RandomRangeVector(Minimum=<-1 -1 -1>, Maximum=<1 1 1.5>), Float=100))
   &SphereLocation(
    MaskSpawnGroup=false
    Override Seed=false
    Sphere Radius=10
    Surface Only=false
    UseNonUniformScale=false
    UseOffset=true
    UseSphereOrientation=false
   )
   &AddVelocityFromPoint(
    Module.Use Falloff Distance=false
    Origin Offset Coordinate Space=::World
    Velocity Strength=&Multiply_Float(A=&RandomRangeFloat(Minimum=360, Maximum=440), B=0.5)
   )
   &CameraOffset(Camera Offset Amount=&RandomRangeFloat(Minimum=2, Maximum=5))
   &AddVelocityFromPoint(
    Module.Use Falloff Distance=false
    Origin Offset=<0 0 -0.1>
    Origin Offset Coordinate Space=::World
    Velocity Origin:NiagaraPosition=$Output.SphereLocation.Position
    Velocity Strength=&RandomRangeFloat(Minimum=150, Maximum=200)
   ) {enabled=false, module=AddVelocityFromPoint001}
   &InheritVelocity(Inherited Velocity Amount Scale=<5000 5000 5000>, Inherited Velocity Speed Limit=50000000188416, Speed Limit=true, UseSpeedThreshold=false) {enabled=false}
   &AddVelocity(Coordinate Space=::World, Velocity=&Multiply_VectorByFloat(Vector=$Engine.Owner.Velocity, Float=1000)) {module=AddVelocity001}
  ]
  particleUpdateScript [
   &InheritSourceMovement(ApplyForce=false, ApplyPosition=true, ApplyPositionFalloff=false)
   &GravityForce(Gravity=<0 0 -490>) {enabled=false}
   &Drag(Drag=3.5, Use Linear Drag=true, Use Rotational Drag=false)
   &SolveForcesAndVelocity(Clamp Velocity=false, Limit Acceleration=false)
   &ParticleState()
   &ScaleColor(
    Scale Alpha=&FloatFromCurve(FloatCurve=DataInterface() {propertyValues {Curve {keys [
         {
          arriveTangentWeight=0.9254
          interpMode=RCIM_Cubic
          leaveTangentWeight=0.9254
          tangentMode=RCTM_User
          tangentWeightMode=RCTWM_WeightedBoth
          time=0
          value=1
         }
         {
          arriveTangent=-7.1278
          arriveTangentWeight=1.1179
          interpMode=RCIM_Cubic
          leaveTangent=-7.1278
          leaveTangentWeight=1.1179
          tangentMode=RCTM_User
          tangentWeightMode=RCTWM_WeightedBoth
          time=1
          value=0
         }
        ]}}})
    ScaleA=true
    ScaleRGB=true
   )
   &ScaleSpriteSize(
    Non-Uniform Curve Sprite Scale=DataInterface() {propertyValues {
      XCurve {keys [<at 0.5 1 RCIM_Cubic>, {arriveTangent=-4.9638, interpMode=RCIM_Cubic, leaveTangent=-4.9638, tangentMode=RCTM_User, time=1, value=0}]}
      YCurve {keys [<at 0.5 1 RCIM_Cubic>, {arriveTangent=-1.8921, interpMode=RCIM_Cubic, leaveTangent=-1.8921, tangentMode=RCTM_User, time=1, value=0.5}]}
     }}
    Scale Sprite Size Mode=::NewEnumerator3
   )
   /SetVariables(Particles.Position=&Add_Vector(
     A=Linked() {linkedVariable {name=Particles.Position, type=NiagaraPosition}}
     B=Linked() {linkedVariable {name=Engine.Owner.Position, type=NiagaraPosition}}
    )) {enabled=false, isSetParameters=true, module=SetVariables_#1}
  ]
  rendererClasses [&NiagaraSpriteRendererProperties]
  renderers [{rendererClass=&NiagaraSpriteRendererProperties}]
 }
 DropsWith_Tail {
  emitterUpdateScript [
   &EmitterState()
   &SpawnBurst_Instantaneous(Spawn Count=3, Use Spawn Probability=false, UseLoopCountLimit=false)
   &SpawnRate(SpawnRate=250, Use Spawn Probability=false) {enabled=false}
   &SpawnPerUnit(Module.Use Max Movement Threshold=true, Module.Use Movement Tolerance=false, Use Spawn Probability=false)
  ]
  particleSpawnScript [
   &InitializeParticle(
    Lifetime=&RandomRangeFloat(Minimum=0.4, Maximum=0.65)
    Position=$Engine.Owner.Position
    Position Mode=::NewEnumerator1
    Sprite Size=&Multiply_Vector2(A=&RandomRangeVector2D(Minimum=<2 4>, Maximum=<4 6>), B=<1.5 5>)
    Sprite Size Mode=::NewEnumerator4
    UsePositionOffset=false
    Write Material Random=true
   )
   &AddVelocity(Velocity=&Multiply_VectorByFloat(Vector=&RandomRangeVector(Minimum=<-1 -1 -1>, Maximum=<1 1 1.5>), Float=100))
   &SphereLocation(
    MaskSpawnGroup=false
    Override Seed=false
    Sphere Radius=10
    Surface Only=false
    UseNonUniformScale=false
    UseOffset=true
    UseSphereOrientation=false
   )
   &AddVelocityFromPoint(
    Module.Use Falloff Distance=false
    Origin Offset Coordinate Space=::World
    Velocity Strength=&RandomRangeFloat(Minimum=360, Maximum=440)
   )
   &CameraOffset(Camera Offset Amount=&RandomRangeFloat(Minimum=2, Maximum=5))
   &AddVelocityFromPoint(
    Module.Use Falloff Distance=false
    Origin Offset=<0 0 -0.1>
    Origin Offset Coordinate Space=::World
    Velocity Origin:NiagaraPosition=$Output.SphereLocation.Position
    Velocity Strength=&RandomRangeFloat(Minimum=150, Maximum=200)
   ) {enabled=false, module=AddVelocityFromPoint001}
   &InheritVelocity(Inherited Velocity Amount Scale=<5000 5000 5000>, Inherited Velocity Speed Limit=50000000188416, Speed Limit=true, UseSpeedThreshold=false) {enabled=false}
   &AddVelocity(Coordinate Space=::World, Velocity=&Multiply_VectorByFloat(Vector=$Engine.Owner.Velocity, Float=1000)) {module=AddVelocity001}
  ]
  particleUpdateScript [
   &InheritSourceMovement(ApplyForce=false, ApplyPosition=true, ApplyPositionFalloff=false)
   &GravityForce(Gravity=<0 0 -490>)
   &SolveForcesAndVelocity(Clamp Velocity=false, Limit Acceleration=false)
   &ParticleState()
   &ScaleColor(
    Scale Alpha=&FloatFromCurve(FloatCurve=DataInterface() {propertyValues {Curve {keys [
         {
          arriveTangentWeight=0.9254
          interpMode=RCIM_Cubic
          leaveTangentWeight=0.9254
          tangentMode=RCTM_User
          tangentWeightMode=RCTWM_WeightedBoth
          time=0
          value=1
         }
         {
          arriveTangent=-7.1278
          arriveTangentWeight=1.1179
          interpMode=RCIM_Cubic
          leaveTangent=-7.1278
          leaveTangentWeight=1.1179
          tangentMode=RCTM_User
          tangentWeightMode=RCTWM_WeightedBoth
          time=1
          value=0
         }
        ]}}})
    ScaleA=true
    ScaleRGB=true
   )
   &ScaleSpriteSize(
    Non-Uniform Curve Sprite Scale=DataInterface() {propertyValues {
      XCurve {keys [{tangentMode=RCTM_User, time=0, value=1}, {tangentMode=RCTM_User, time=1, value=0}]}
      YCurve {keys [{tangentMode=RCTM_User, time=0, value=1}, {tangentMode=RCTM_User, time=1, value=1}]}
     }}
    Scale Sprite Size Mode=::NewEnumerator3
   )
   /SetVariables(Particles.Position=&Add_Vector(
     A=Linked() {linkedVariable {name=Particles.Position, type=NiagaraPosition}}
     B=Linked() {linkedVariable {name=Engine.Owner.Position, type=NiagaraPosition}}
    )) {enabled=false, isSetParameters=true, module=SetVariables_#1}
   &GenerateLocationEvent(Event Send Rate=90, Use Event Probability=false, UseEventDelay=false)
  ]
  rendererClasses [&NiagaraSpriteRendererProperties]
  renderers [{rendererClass=&NiagaraSpriteRendererProperties}]
 }
 Drops_Curvyshockwave {
  emitterUpdateScript [
   &EmitterState()
   &SpawnBurst_Instantaneous(Spawn Count=1, Use Spawn Probability=false)
   &SpawnBurst_Instantaneous(Spawn Count=1, Spawn Group=1, Spawn Time=0.05, Use Spawn Probability=false) {module=SpawnBurst_Instantaneous001}
  ]
  particleSpawnScript [
   &InitializeParticle(
    Color=<rgba 1 1 1 3>
    Lifetime=&RandomRangeFloat(Minimum=0.35, Maximum=0.6)
    Sprite Rotation Angle=&Lerp_Float(A=0.25, B=1, Alpha=&MakeFloatFromInt(Int=&SpawnGroup()))
    Sprite Rotation Mode=::NewEnumerator0
    Sprite Size=<240 136.9728>
    Sprite Size Mode=::NewEnumerator4
    Sprite UV Mode=::NewEnumerator1
    UsePositionOffset=false
    Write Material Random=true
   )
   &SphereLocation(
    MaskSpawnGroup=false
    Override Seed=false
    Sphere Radius=0
    Surface Only=false
    UseNonUniformScale=false
    UseOffset=true
    UseSphereOrientation=false
   ) {enabled=false}
   &CameraOffset(Camera Offset Amount=&RandomRangeFloat(Minimum=25, Maximum=25))
   &AddVelocity(Velocity=&SelectVectorFromArray(
     Direct Array Index=&ReturnExecIndex()
     Vector Selection Array=DataInterface() {propertyValues {FloatData [<0 0 -2000>, <0 0 2000>]}}
    )) {enabled=false, module=AddVelocity001}
   &AddVelocity(Coordinate Space=::World, Velocity=<0 0 -2500>) {enabled=false}
   /SetVariables(Particles.SpriteRotation:NiagaraFloat=&Lerp_Float(A=120, B=-120, Alpha=&MakeFloatFromInt(Int=$Emitter.SpawnGroup))) {isSetParameters=true}
  ]
  particleUpdateScript [
   &ParticleState()
   &SpriteFacingAndAlignment(Sprite Alignment:NiagaraPosition=$Particles.Position, Sprite Facing=<0 0 1>, WriteAlignment=true, WriteFacing=true) {enabled=false}
   &Drag(Drag=7, Use Linear Drag=true, Use Rotational Drag=false)
   &SolveForcesAndVelocity(Clamp Velocity=false, Limit Acceleration=false)
   &DynamicMaterialParameters(
    Index 0 Param 1=&FloatFromCurve(FloatCurve=DataInterface() {propertyValues {Curve {keys [
         {
          arriveTangent=-0.2664
          arriveTangentWeight=0.2068
          interpMode=RCIM_Cubic
          leaveTangent=-0.2664
          leaveTangentWeight=0.2068
          tangentMode=RCTM_User
          tangentWeightMode=RCTWM_WeightedBoth
          time=0.25
          value=1.1226
         }
         {
          arriveTangent=-0.7856
          arriveTangentWeight=0.5298
          interpMode=RCIM_Cubic
          leaveTangent=-0.311
          leaveTangentWeight=0.0332
          tangentMode=RCTM_Break
          tangentWeightMode=RCTWM_WeightedBoth
          time=0.9
          value=0.2067
         }
        ]}}})
    Index 0 Param 2=0.05
    Index 0 Param 3=&RandomRangeFloat(Minimum=0, Maximum=1)
    Index 0 Param 4=&RandomRangeFloat(Minimum=0, Maximum=1)
    Index 1 Param 1=1
    Index 1 Param 2=1
    Index 2 Param 2=1
    Index 2 Param 4=-0.1
    Index 3 Param 1=2
    Index 3 Param 2=0.1
    Index 3 Param 3=1
    Index 3 Param 4=1
    Module.Param0WriteEnabled=true
    Module.Param0WriteEnabled1=true
    Module.Param0WriteEnabled2=true
    Module.Param0WriteEnabled3=true
    Module.Param1WriteEnabled=true
    Module.Param1WriteEnabled1=true
    Module.Param1WriteEnabled2=true
    Module.Param1WriteEnabled3=true
    Module.Param2WriteEnabled=true
    Module.Param2WriteEnabled1=true
    Module.Param2WriteEnabled2=true
    Module.Param2WriteEnabled3=true
    Module.Param3WriteEnabled=true
    Module.Param3WriteEnabled1=true
    Module.Param3WriteEnabled2=true
    Module.Param3WriteEnabled3=true
    Write Parameter Index 1=true
    Write Parameter Index 2=true
    Write Parameter Index 3=true
   )
   &ScaleColor(
    Scale Alpha=&FloatFromCurve(FloatCurve=DataInterface() {propertyValues {Curve {keys [{arriveTangentWeight=0.3986, leaveTangentWeight=0.0498, time=0.85, value=1}, {arriveTangentWeight=0.0333, time=1, value=0}]}}})
    ScaleA=true
    ScaleRGB=true
   )
   &ScaleSpriteSize(Scale Factor=&Vector2DFromFloat(Value=&FloatFromCurve(
      FloatCurve=DataInterface() {propertyValues {Curve {keys [
          {arriveTangent=1.1628, interpMode=RCIM_Cubic, leaveTangent=1.1628, tangentMode=RCTM_User, time=0, value=0.1}
          {arriveTangent=0.5523, interpMode=RCIM_Cubic, leaveTangent=0.5523, tangentMode=RCTM_User, time=1, value=1}
         ]}}}
      Scale Curve=&RandomRangeFloat(Minimum=1, Maximum=1.25)
     )))
  ]
  rendererClasses [&NiagaraSpriteRendererProperties]
  renderers [{rendererClass=&NiagaraSpriteRendererProperties}]
 }
 Lightning_Spike_Arcs {
  emitterUpdateScript [
   &EmitterState(Life Cycle Mode=::NewEnumerator1, Loop Behavior=::NewEnumerator1, Loop Duration=0.5, UseLoopDelay=false)
   &SpawnBurst_Instantaneous(Use Spawn Probability=false, UseLoopCountLimit=false)
   &SpawnRate(
    SpawnRate=&FloatFromCurve(
     FloatCurve=DataInterface() {propertyValues {Curve {keys [
         <at 0 1 RCIM_Cubic>
         {arriveTangent=-1.6667, interpMode=RCIM_Cubic, leaveTangent=-1.6667, time=0.2, value=0}
         <at 0.6 0>
         <at 0.7 0.8>
         <at 1 0>
        ]}}}
     CurveIndex=$Emitter.NormalizedLoopAge
     Scale Curve=10
    )
    Use Spawn Probability=false
   )
  ]
  particleSpawnScript [
   &InitializeParticle(
    Color=<rgba 2 0 1.5333 1>
    Lifetime Max=0.125
    Lifetime Min=0.1
    Lifetime Mode=::NewEnumerator1
    Mesh Scale Max=<0.037 7 7>
    Mesh Scale Min=<0.037 2 2>
    Mesh Scale Mode=::NewEnumerator2
    Position Mode=::NewEnumerator1
    Sprite Size Mode=::NewEnumerator3
    Uniform Sprite Size=31.3334
    UsePositionOffset=false
    Write Material Random=true
   )
   /SetVariables(Particles.Pos Choice:NiagaraInt32=&Multiply_Int(A=&RandomRangeInt(Minimum=7, Maximum=20), B=2)) {enabled=false, isSetParameters=true}
   /SetVariables(Particles.Pos Choice:NiagaraInt32=&Multiply_Int(
     A=&SelectIntFromArray(
      Array Sampling Mode=::NewEnumerator1
      Direct Array Index=?
      Direct Array Mode=::NewEnumerator1
      Int Selection Array=DataInterface() {propertyValues {IntData [0,1,2,3,4,6]}}
      Random Seed=0
     )
     B=2
    )) {isSetParameters=true}
   &InitialMeshOrientation(
    Mesh Orientation Mode=::NewEnumerator7
    Rotation=&RandomRangeVector(Minimum=<0 0 0>, Maximum=<1 0 0>)
    Use Rotation Vector=true
   )
   /SetVariables(Particles.Pos A Color:LinearColor=<rgba 1 0 0 1>, Particles.PosB COlor:LinearColor=<rgba 0 1 0 1>) {isSetParameters=true}
  ]
  particleUpdateScript [
   &ParticleState()
   &self:Drift_Horror_Get_Set_of_4_Bones_02(New Skeletal Mesh:NiagaraDataInterfaceSkeletalMesh=$User.SKeletalMesh, StartIndex:NiagaraInt32=$Particles.Initial.Pos Choice) {module=Drift_Horror_Get_Set_of_4_Bones_02}
   /SetVariables(Particles.Pos A:Vector3f=$Output.Drift_Horror_Get_Set_of_4_Bones_02.BonePos1) {isSetParameters=true}
   /SetVariables(Particles.Position:Vector3f=$Particles.Pos A) {isSetParameters=true}
   &DynamicMaterialParameters(
    Index 0 Param 1=1
    Index 0 Param 2=0.5
    Index 0 Param 3=-0.1
    Module.Param0WriteEnabled=true
    Module.Param1WriteEnabled=true
    Module.Param2WriteEnabled=true
    Module.Param3WriteEnabled=true
   )
   /Niagara/Modules/Update/Orientation/V2/OrientMeshToVector(
    Look At Direction:Vector3f=&NormalizeVector(
     Vector To Normalize=&Subtract_Vector(A=$Particles.Pos B, B=$Particles.Pos A)
     Safe Normalize=true
     Safe Fallback Threshold=0
     Safe Fallback Vector=<0 0 0.001>
    )
    Override Quat=false
    Up Axis Direction:Vector3f=<0 0 1>
    UseOptionalSideAxis=false
    UseOptionalUpAxis=false
   )
   &ScaleMeshSize(Scale Factor=&MakeVector(
     X=&VectorLength(Vector=&Subtract_Vector(A=$Particles.Pos B, B=$Particles.Pos A))
     Y=&Lerp_Float(A=1, B=&RandomRangeFloat(Minimum=0, Maximum=1.5), Alpha=$Particles.NormalizedAge)
     Z=&Lerp_Float(A=1, B=&RandomRangeFloat(Minimum=0.5, Maximum=1.5), Alpha=$Particles.NormalizedAge)
    ))
   &FN_Light_Attributes(
    Color=<rgba 25 6.25 20.4806 1>
    Create Color Attribute=true
    Create Exponent Attribute=true
    Create Position Attribute=true
    Position=&Lerp_Vector(A=$Particles.Pos A, B=$Particles.Pos B, Alpha=0.5)
    Scale Color=&FloatFromCurve(FloatCurve=DataInterface() {propertyValues {Curve {keys [
         {arriveTangent=5.7658, interpMode=RCIM_Cubic, leaveTangent=5.7658, tangentMode=RCTM_User, time=0, value=0}
         <at 0.3 0.9>
         {interpMode=RCIM_Cubic, tangentMode=RCTM_User, time=0.5, value=0.5}
         <at 0.8 0.8>
         {arriveTangent=-4.8787, interpMode=RCIM_Cubic, leaveTangent=-4.8787, tangentMode=RCTM_User, time=1, value=0}
        ]}}})
   )
   &ScaleColor(
    Scale Alpha=&Multiply_Float(
     A=&FloatFromCurve(FloatCurve=DataInterface() {propertyValues {Curve {keys [
          <at 0 1 RCIM_Cubic>
          {arriveTangent=-0.0405, interpMode=RCIM_Cubic, leaveTangent=-0.0405, tangentMode=RCTM_User, time=0.7009, value=1}
          {arriveTangent=-8.0089, interpMode=RCIM_Cubic, leaveTangent=-8.0089, tangentMode=RCTM_User, time=1, value=0}
         ]}}})
     B=&Sine(
      Normalized Angle=$Emitter.Age
      Period=&RandomRangeFloat(Minimum=0.05, Maximum=0.2)
      Scale=2
      Bias=&RandomRangeFloat(Minimum=0.5, Maximum=3)
     )
    )
    ScaleA=true
    ScaleRGB=true
   )
  ]
  rendererClasses [&NiagaraMeshRendererProperties, &NiagaraLightRendererProperties]
  renderers [
   {rendererClass=&NiagaraMeshRendererProperties, rendererIndex=0}
   {rendererClass=&NiagaraLightRendererProperties, rendererIndex=1}
  ]
 }
 Lightning_ou_of_spikes001 {
  emitterUpdateScript [
   &EmitterState(Life Cycle Mode=::NewEnumerator1, Loop Behavior=::NewEnumerator1, Loop Duration=0.6, UseLoopDelay=false)
   &SpawnBurst_Instantaneous(
    Spawn Count=&RandomRangeInt(Minimum=100, Maximum=90)
    Use Spawn Probability=false
    UseLoopCountLimit=false
   ) {enabled=false}
   &SpawnRate(
    SpawnRate=&FloatFromCurve(
     FloatCurve=DataInterface() {propertyValues {Curve {keys [
         {
          arriveTangent=-2.4215
          arriveTangentWeight=0.711
          interpMode=RCIM_Cubic
          leaveTangent=-2.4215
          leaveTangentWeight=0.711
          tangentMode=RCTM_User
          tangentWeightMode=RCTWM_WeightedBoth
          time=0
          value=1
         }
         {
          arriveTangent=-0.2546
          arriveTangentWeight=0.5621
          interpMode=RCIM_Cubic
          leaveTangent=-0.2546
          leaveTangentWeight=0.5621
          tangentMode=RCTM_User
          tangentWeightMode=RCTWM_WeightedBoth
          time=1
          value=0
         }
        ]}}}
     CurveIndex=$Emitter.NormalizedLoopAge
     Scale Curve=150
    )
    Use Spawn Probability=false
   )
  ]
  particleSpawnScript [
   &InitializeParticle(
    Color=<rgba 15 0 11.5 25>
    Lifetime Max=0.125
    Lifetime Min=0.125
    Lifetime Mode=::NewEnumerator1
    Mesh Scale Min=<4 4 4>
    Mesh Scale Mode=::NewEnumerator2
    Position Mode=::NewEnumerator1
    Position Offset=<0 -20 0>
    Sprite Size Mode=::NewEnumerator3
    Uniform Sprite Size=2
    UsePositionOffset=true
    Write Material Random=true
   )
   &InitialMeshOrientation(Use Rotation Vector=true)
   &SphereLocation(
    MaskSpawnGroup=false
    Override Seed=false
    Sphere Radius=20
    Surface Only=false
    UseNonUniformScale=false
    UseOffset=false
    UseSphereOrientation=false
   )
  ]
  particleUpdateScript [
   &ParticleState()
   &DynamicMaterialParameters(
    Index 0 Param 1=1
    Index 0 Param 2=-0.9
    Index 0 Param 3=-0.1
    Index 0 Param 4=0.5
    Module.Param0WriteEnabled=true
    Module.Param1WriteEnabled=true
    Module.Param2WriteEnabled=true
    Module.Param3WriteEnabled=true
   )
   &ScaleMeshSize(Scale Factor=&MakeVector(
     X=&FloatFromCurve(FloatCurve=DataInterface() {propertyValues {Curve {keys [{tangentMode=RCTM_User, time=0, value=0}, {tangentMode=RCTM_User, time=0.1, value=1}, <at 1 0.9>]}}})
     Y=&Lerp_Float(A=1, B=&RandomRangeFloat(Minimum=0.5, Maximum=1.5), Alpha=$Particles.NormalizedAge)
     Z=&Lerp_Float(A=1, B=&RandomRangeFloat(Minimum=0.5, Maximum=1.5), Alpha=$Particles.NormalizedAge)
    )) {module=ScaleMeshSize001}
   &FN_Light_Attributes(
    Color=<rgba 5 1.25 4.0961 1>
    Create Color Attribute=true
    Create Exponent Attribute=true
    Create Position Attribute=true
    Enable Particle Light Enable=true
    Light Enabled:NiagaraBool=&SetBoolByFloatComparison(A=&RandomRangeFloat(Minimum=0, Maximum=1), B=0.2, Comparison Type=::NewEnumerator6)
    Position:NiagaraPosition=$Particles.Position
   ) {enabled=false}
   &ScaleColor(
    Scale Alpha=&Sine(
     Normalized Angle=$Emitter.Age
     Period=&RandomRangeFloat(Minimum=0.05, Maximum=0.2)
     Scale=2
     Bias=&RandomRangeFloat(Minimum=0.5, Maximum=3)
    )
    ScaleA=true
    ScaleRGB=true
   )
   &UpdateMeshOrientation(
    Rotation Coordinate Space=::NewEnumerator3
    Rotation Rate=&RandomRangeFloat(Minimum=-0.1, Maximum=0.1)
    Rotation Vector=&RandomVector(Vector Scale=1)
   )
  ]
  rendererClasses [&NiagaraMeshRendererProperties, &NiagaraLightRendererProperties]
  renderers [
   {rendererClass=&NiagaraMeshRendererProperties, rendererIndex=0}
   {rendererClass=&NiagaraLightRendererProperties, rendererIndex=1}
  ]
 }
 Lightning_ou_of_spikes_hero {
  emitterUpdateScript [
   &EmitterState(Life Cycle Mode=::NewEnumerator1, Loop Behavior=::NewEnumerator1, Loop Duration=0.765, UseLoopDelay=false)
   &SpawnBurst_Instantaneous(Spawn Count=3, Use Spawn Probability=false, UseLoopCountLimit=false)
   &SpawnRate(
    SpawnRate=&FloatFromCurve(
     FloatCurve=DataInterface() {propertyValues {Curve {keys [
         {
          arriveTangent=-2.4215
          arriveTangentWeight=0.711
          interpMode=RCIM_Cubic
          leaveTangent=-2.4215
          leaveTangentWeight=0.711
          tangentMode=RCTM_User
          tangentWeightMode=RCTWM_WeightedBoth
          time=0
          value=1
         }
         {
          arriveTangent=-0.2546
          arriveTangentWeight=0.5621
          interpMode=RCIM_Cubic
          leaveTangent=-0.2546
          leaveTangentWeight=0.5621
          tangentMode=RCTM_User
          tangentWeightMode=RCTWM_WeightedBoth
          time=1
          value=0
         }
        ]}}}
     CurveIndex=$Emitter.NormalizedLoopAge
     Scale Curve=250
    )
    Use Spawn Probability=false
   ) {enabled=false}
   &SpawnBurst_Instantaneous(Spawn Count=5, Spawn Time=0.765, Use Spawn Probability=false, UseLoopCountLimit=false) {module=SpawnBurst_Instantaneous001}
  ]
  particleSpawnScript [
   &InitializeParticle(
    Color=<rgba 15 0 11.5 5>
    Lifetime Max=0.125
    Lifetime Min=0.125
    Lifetime Mode=::NewEnumerator1
    Mesh Scale Max=<4 4 4>
    Mesh Scale Min=<5 5 5>
    Mesh Scale Mode=::NewEnumerator2
    Position=/Niagara/DynamicInputs/Helpers/SimulationPosition()
    Position Mode=::NewEnumerator1
    Position Offset=<0 -20 0>
    Position Offset Coordinate Space=::Local
    Sprite Size Mode=::NewEnumerator3
    Uniform Sprite Size=2
    UsePositionOffset=true
    Write Material Random=true
   )
   &InitialMeshOrientation(Use Rotation Vector=true)
   &SphereLocation(
    MaskSpawnGroup=false
    Override Seed=false
    Sphere Radius=20
    Surface Only=false
    UseNonUniformScale=false
    UseOffset=false
    UseSphereOrientation=false
   )
  ]
  particleUpdateScript [
   &ParticleState()
   &DynamicMaterialParameters(
    Index 0 Param 1=1
    Index 0 Param 2=-0.9
    Index 0 Param 3=-0.1
    Index 0 Param 4=0.5
    Module.Param0WriteEnabled=true
    Module.Param1WriteEnabled=true
    Module.Param2WriteEnabled=true
    Module.Param3WriteEnabled=true
   )
   &ScaleMeshSize(Scale Factor=&MakeVector(
     X=&FloatFromCurve(FloatCurve=DataInterface() {propertyValues {Curve {keys [{tangentMode=RCTM_User, time=0, value=0}, {tangentMode=RCTM_User, time=0.1, value=1}, <at 0.9879 1.1484>]}}})
     Y=&Lerp_Float(A=1, B=&RandomRangeFloat(Minimum=0.5, Maximum=1.25), Alpha=$Particles.NormalizedAge)
     Z=&Lerp_Float(A=1, B=&RandomRangeFloat(Minimum=0.5, Maximum=1.25), Alpha=$Particles.NormalizedAge)
    )) {module=ScaleMeshSize001}
   &FN_Light_Attributes(
    Color=<rgba 25 6.25 20.4807 1>
    Create Color Attribute=true
    Create Exponent Attribute=true
    Create Position Attribute=true
    Enable Particle Light Enable=true
    Light Enabled:NiagaraBool=&SetBoolByFloatComparison(A=&RandomRangeFloat(Minimum=0, Maximum=1), B=0.2, Comparison Type=::NewEnumerator6)
    Position=&Add_Vector(
     A=Linked() {linkedVariable {name=Particles.Position, type=NiagaraPosition}}
     B=&Multiply_VectorByFloat(
      Vector=&ReturnMeshOrientationAxis(Mesh Orientation Axis=::XAxis, Mesh Orientation Quat=$Particles.MeshOrientation)
      Float=-1
     )
    )
   )
   &ScaleColor(
    Scale Alpha=&Multiply_Float(
     A=&FloatFromCurve(FloatCurve=DataInterface() {propertyValues {Curve {keys [<at 0 1>,<at 1 0>]}}})
     B=&Sine(Normalized Angle=$Emitter.Age, Period=0.1, Scale=0.5, Bias=0.5)
    )
    ScaleA=true
    ScaleRGB=true
   )
   &InheritSourceMovement(ApplyForce=false, ApplyPosition=true, ApplyPositionFalloff=false)
   &SolveForcesAndVelocity(Clamp Velocity=false, Limit Acceleration=false)
  ]
  rendererClasses [&NiagaraMeshRendererProperties, &NiagaraLightRendererProperties]
  renderers [
   {rendererClass=&NiagaraMeshRendererProperties, rendererIndex=0}
   {rendererClass=&NiagaraLightRendererProperties, rendererIndex=1}
  ]
 }
 Slurpy_Streaks {
  emitterUpdateScript [
   &EmitterState()
   &SpawnBurst_Instantaneous(Spawn Count=3, Use Spawn Probability=false)
   &SpawnRate(SpawnRate=50, Use Spawn Probability=false) {enabled=false}
  ]
  particleSpawnScript [
   &InitializeParticle(
    Lifetime Max=0.65
    Lifetime Min=0.4
    Lifetime Mode=::NewEnumerator1
    Sprite Rotation Mode=::NewEnumerator2
    Sprite Size Max=<40 101.25>
    Sprite Size Min=<20 121.5>
    Sprite Size Mode=::NewEnumerator2
    Sprite UV Mode=::NewEnumerator3
    UsePositionOffset=false
    Write Material Random=true
   ) {module=InitializeParticle001}
   &SphereLocation(
    MaskSpawnGroup=false
    Override Seed=false
    Sphere Radius=5
    Surface Only=false
    UseNonUniformScale=false
    UseOffset=true
    UseSphereOrientation=false
   )
   &CameraOffset(Camera Offset Amount=&RandomRangeFloat(Minimum=35, Maximum=50))
   &SubUVAnimation(End Frame=2, Start Frame=2, SubUV Animation Mode:ENiagara_SubUVLookupMode=::NewEnumerator2)
   /SetVariables(Particles.Velocity=&RandomVector(Vector Scale=25)) {isSetParameters=true, module=SetVariables_#2}
  ]
  particleUpdateScript [
   &ParticleState()
   &ScaleSpriteSize(Scale Factor=&MakeVector2D(
     X=2
     Y=&FloatFromCurve(
      FloatCurve=DataInterface() {propertyValues {Curve {keys [
          {arriveTangent=1.9742, interpMode=RCIM_Cubic, leaveTangent=1.9742, tangentMode=RCTM_User, time=0, value=0}
          {arriveTangent=0.2346, interpMode=RCIM_Cubic, leaveTangent=0.2346, tangentMode=RCTM_User, time=1, value=0.9}
         ]}}}
      Scale Curve=1.25
     )
    ))
   &DynamicMaterialParameters(
    Index 0 Param 1=&FloatFromCurve(FloatCurve=DataInterface() {propertyValues {Curve {keys [
         {interpMode=RCIM_Cubic, leaveTangentWeight=0.3367, tangentMode=RCTM_User, time=0, value=1}
         {arriveTangent=-2.0812, interpMode=RCIM_Cubic, leaveTangent=-2.0812, tangentMode=RCTM_User, time=1, value=0}
        ]}}})
    Index 0 Param 3=1
    Index 0 Param 4=1
    Module.Param0WriteEnabled=true
    Module.Param1WriteEnabled=true
    Module.Param2WriteEnabled=true
    Module.Param3WriteEnabled=true
   )
   &ScaleColor(
    Scale Alpha=&FloatFromCurve(
     FloatCurve=DataInterface() {propertyValues {Curve {keys [
         {interpMode=RCIM_Cubic, tangentMode=RCTM_User, time=0, value=0}
         {interpMode=RCIM_Cubic, tangentMode=RCTM_User, time=0.1, value=1}
         {arriveTangent=-1.1111, interpMode=RCIM_Cubic, leaveTangent=-1.1111, time=0.8, value=0.95}
         {arriveTangent=-3.0227, interpMode=RCIM_Cubic, leaveTangent=-3.0227, tangentMode=RCTM_User, time=1, value=0}
        ]}}}
     Scale Curve=2
    )
    ScaleA=true
    ScaleRGB=true
   ) {module=ScaleColor001}
   /SetVariables() {enabled=false, isSetParameters=true, module=SetVariables_#3}
   &InheritSourceMovement(ApplyForce=false, ApplyPosition=true, ApplyPositionFalloff=false)
   &SolveForcesAndVelocity(Clamp Velocity=false, Limit Acceleration=false)
  ]
  rendererClasses [&NiagaraSpriteRendererProperties]
  renderers [{rendererClass=&NiagaraSpriteRendererProperties}]
 }
 Slurpy_Streaks001 {
  emitterUpdateScript [
   &EmitterState()
   &SpawnBurst_Instantaneous(Spawn Count=3, Use Spawn Probability=false)
   &SpawnRate(SpawnRate=50, Use Spawn Probability=false) {enabled=false}
  ]
  particleSpawnScript [
   &InitializeParticle(
    Color=<rgba 0.5 0.5 0.5 1>
    Lifetime Max=0.5
    Lifetime Min=0.4
    Lifetime Mode=::NewEnumerator1
    Sprite Rotation Mode=::NewEnumerator2
    Sprite Size Max=<36 101.25>
    Sprite Size Min=<14.4 121.5>
    Sprite Size Mode=::NewEnumerator2
    Sprite UV Mode=::NewEnumerator3
    UsePositionOffset=true
    Write Material Random=true
   ) {module=InitializeParticle001}
   &SphereLocation(
    MaskSpawnGroup=false
    Override Seed=false
    Sphere Radius=5
    Surface Only=false
    UseNonUniformScale=false
    UseOffset=true
    UseSphereOrientation=false
   )
   &CameraOffset(Camera Offset Amount=45)
   &SubUVAnimation(End Frame=2, Start Frame=2, SubUV Animation Mode:ENiagara_SubUVLookupMode=::NewEnumerator2)
   /SetVariables(Particles.Velocity=&RandomVector(Vector Scale=1)) {isSetParameters=true, module=SetVariables_#2}
  ]
  particleUpdateScript [
   &ParticleState()
   &ScaleSpriteSize(Scale Factor=&MakeVector2D(
     X=2
     Y=&FloatFromCurve(
      FloatCurve=DataInterface() {propertyValues {Curve {keys [
          {arriveTangent=1.9946, interpMode=RCIM_Cubic, leaveTangent=1.9946, tangentMode=RCTM_User, time=0, value=0}
          {arriveTangent=0.283, interpMode=RCIM_Cubic, leaveTangent=0.283, tangentMode=RCTM_User, time=1, value=0.9}
         ]}}}
      Scale Curve=1.25
     )
    ))
   &DynamicMaterialParameters(
    Index 0 Param 1=&FloatFromCurve(FloatCurve=DataInterface() {propertyValues {Curve {keys [
         {
          arriveTangent=-2.2914
          interpMode=RCIM_Cubic
          leaveTangentWeight=0.3333
          tangentMode=RCTM_User
          tangentWeightMode=RCTWM_WeightedBoth
          time=0
          value=1
         }
         {arriveTangent=-3.6318, interpMode=RCIM_Cubic, leaveTangent=-3.6318, tangentMode=RCTM_User, time=1, value=0}
        ]}}})
    Index 0 Param 3=1
    Index 0 Param 4=1
    Module.Param0WriteEnabled=true
    Module.Param1WriteEnabled=true
    Module.Param2WriteEnabled=true
    Module.Param3WriteEnabled=true
   )
   &ScaleColor(
    Scale Alpha=&FloatFromCurve(
     FloatCurve=DataInterface() {propertyValues {Curve {keys [
         {interpMode=RCIM_Cubic, tangentMode=RCTM_User, time=0, value=0}
         {interpMode=RCIM_Cubic, tangentMode=RCTM_User, time=0.1, value=1}
         {arriveTangent=-1.1111, interpMode=RCIM_Cubic, leaveTangent=-1.1111, time=0.8, value=0.95}
         {arriveTangent=-3.0227, interpMode=RCIM_Cubic, leaveTangent=-3.0227, tangentMode=RCTM_User, time=1, value=0}
        ]}}}
     Scale Curve=2
    )
    ScaleA=true
    ScaleRGB=true
   ) {module=ScaleColor001}
   /SetVariables() {enabled=false, isSetParameters=true, module=SetVariables_#3}
   &InheritSourceMovement(ApplyForce=false, ApplyPosition=true, ApplyPositionFalloff=false)
   &SolveForcesAndVelocity(Clamp Velocity=false, Limit Acceleration=false)
  ]
  rendererClasses [&NiagaraSpriteRendererProperties]
  renderers [{rendererClass=&NiagaraSpriteRendererProperties}]
 }
 Tail {
  emitterUpdateScript [
   &EmitterState(Life Cycle Mode=::NewEnumerator1, Loop Duration=0.25, UseLoopDelay=false)
   /SetVariables() {isSetParameters=true}
  ]
  particleSpawnScript [&InitializeRibbon(Color=<rgba 0 0 0 5>, Lifetime=0.125, Ribbon Width=10, UsePositionOffset=false, Write Material Random=true)]
  particleUpdateScript [
   &ParticleState()
   &InheritSourceMovement(Applied Position Scale=<0.9 0.9 0.9>, ApplyForce=false, ApplyPosition=true, ApplyPositionFalloff=false)
   &ScaleColor(
    Scale Alpha=&FloatFromCurve(FloatCurve=DataInterface() {propertyValues {Curve {keys [<at 0 1>,<at 1 0>]}}})
    ScaleA=true
    ScaleRGB=true
   )
   &CurlNoiseForce(Noise Strength=&Multiply_Float(A=500, B=$Particles.NormalizedAge), Pan Noise=false) {enabled=false}
   &SolveForcesAndVelocity(Clamp Velocity=false, Limit Acceleration=false)
   &ScaleRibbonWidth(Ribbon Width Scale=&FloatFromCurve(FloatCurve=DataInterface() {propertyValues {Curve {keys [<at 0 1>, {leaveTangent=-7.8977, time=0.8, value=0}]}}}, CurveIndex=$Emitter.Age))
  ]
  rendererClasses [&NiagaraSpriteRendererProperties, &NiagaraRibbonRendererProperties]
  renderers [
   {rendererClass=&NiagaraSpriteRendererProperties}
   {rendererClass=&NiagaraRibbonRendererProperties, rendererIndex=1}
  ]
 }
}
summary {
 systemName=NS_DriftHorror_Glider_Deply
 emitters [
  {emitterName=Drops, rendererClasses [&NiagaraSpriteRendererProperties, &NiagaraMeshRendererProperties]}
  {emitterName=Slurpy_Streaks, rendererClasses [&NiagaraSpriteRendererProperties]}
  {emitterName=Drops_Curvyshockwave, rendererClasses [&NiagaraSpriteRendererProperties]}
  {emitterName=DropsWith_Tail, rendererClasses [&NiagaraSpriteRendererProperties]}
  {emitterName=Tail, rendererClasses [&NiagaraSpriteRendererProperties, &NiagaraRibbonRendererProperties]}
  {emitterName=Drops001, rendererClasses [&NiagaraSpriteRendererProperties]}
  {emitterName=Slurpy_Streaks001, rendererClasses [&NiagaraSpriteRendererProperties]}
  {
   emitterName=Lightning_ou_of_spikes001
   rendererClasses [&NiagaraMeshRendererProperties, &NiagaraLightRendererProperties]
  }
  {
   emitterName=Lightning_ou_of_spikes_hero
   rendererClasses [&NiagaraMeshRendererProperties, &NiagaraLightRendererProperties]
  }
  {
   emitterName=Lightning_Spike_Arcs
   rendererClasses [&NiagaraMeshRendererProperties, &NiagaraLightRendererProperties]
  }
 ]
 userVariables [
  User.RibbonIDOffset {type=NiagaraInt32}
  User.SkeletalMesh {
   type=NiagaraDataInterfaceSkeletalMesh
   defaultValue {
    struct=NiagaraExt_VariableValue_DataInterface
    value {dataInterface=&self:NiagaraDataInterfaceSkeletalMesh_0}
   }
  }
 ]
}
